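# Editor utility with batch tools for preparing selected tree static meshes.
# BakeTreeImpostor opens a Tree Pipeline widget tab for the first selected mesh, though its widget cast is a broken node.
# SetCollisionFromMesh uses Geometry Script to build convex collision on selected meshes from the one named "collision".
# SetTreeProperties sets collision LOD, then uses embedded Python for Nanite, shape preservation, Lerp UVs and distance field settings, and shows a summary dialog.
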EditorUtilityBlueprint /Game/NeverCook/ArtTools/TreeTools/BPUtility_TreePipeline
graphs [
 BakeTreeImpostor {nodes [
   BakeTreeImpostor() {is=title K2Node_FunctionEntry}
   > n1 = GetSelectedAssets()
   > Branch(Condition=Is Valid Index!(TargetArray:Array of Object References=@n1, IndexToTest:int) -> (ReturnValue:bool))
   > n2 = Cast To StaticMesh(Object="Get (a copy)"(Array:Array of Object References=@n1) -> (Output:Object Reference))
   > n3 = SpawnAndRegisterTab(self=Get EditorUtilitySubsystem())
   > n4 = Bad cast node!(execute, Object=@n3) -> (then, CastFailed, AsEUW Tree Pipeline:Object Reference) {is=title K2Node_DynamicCast}
   > Set Static Mesh!(execute, Static Mesh:Static Mesh Object Reference=@n2, self:Object Reference=@n4) -> (then, Output_Get:Static Mesh Object Reference)
   > Make Names(self:Object Reference=@n4)
  ]}
 EventGraph {links [], nodes []}
 SetCollisionFromMesh {nodes [
   SetCollisionFromMesh!() -> (then, CollisionMesh:Static Mesh Object Reference) {is=title K2Node_FunctionEntry}
   > n1 = Sequence!(execute) -> (then_0, then_1, then_2:Exec)
   > n2 = GetSelectedAssets()
   > n3 = For Each Loop(Array:Array of Object References=@n2) -> (Array Element:Object Reference)
   > n4 = Cast To StaticMesh(Object=@n3)
   > Branch(Condition=Contains(SearchIn=GetDisplayName(Object=@n4), Substring=collision))
   > Set Collision(Collision:Static Mesh Object Reference=@n4) -> (Output_Get:Static Mesh Object Reference)
   n5 = Get!(execute:Exec=@n1.then_1) -> (then:Exec, else:Exec, CollisionMesh:Static Mesh Object Reference)
   > Set Collision(Collision=@n5)
   n6 = GetSelectedAssets(execute=@n1.then_2)
   > n7 = For Each Loop(Array=@n6)
   > n8 = Cast To StaticMesh(Object=@n7)
   > Branch(Condition="Not Equal (Object)"(A:Object Reference=@n8, B:Object Reference=Get Collision() -> (Collision:Static Mesh Object Reference)) -> (ReturnValue:bool=~) {is=title K2Node_PromotableOperator})
   > n9 = CreateDynamicMeshPool!(execute) -> (then, ReturnValue:Dynamic Mesh Pool Object Reference)
   > n10 = RequestMesh!(execute, self:Dynamic Mesh Pool Object Reference=@n9) -> (then, ReturnValue:Dynamic Mesh Object Reference)
   > n11 = Copy Mesh From Static Mesh!(
    execute
    FromStaticMeshAsset:Static Mesh Object Reference=Get Collision()
    ToDynamicMesh:Dynamic Mesh Object Reference=@n10
    AssetOptions:Geometry Script Copy Mesh From Asset Options Structure
    RequestedLOD:Geometry Script Mesh Read LOD Structure
    bUseSectionMaterials:bool=true
   ) -> (Failure:Exec, Success:Exec, ReturnValue:Dynamic Mesh Object Reference)
   SetStaticMeshCollisionFromMesh!(
    execute=@n11.Success
    FromDynamicMesh:Dynamic Mesh Object Reference=@n11
    ToStaticMeshAsset:Static Mesh Object Reference=@n8
    Options:Geometry Script Collision From Mesh Options Structure=Make Geometry Script Collision from Mesh Options!(
     bEmitTransaction:bool=True
     Method:EGeometryScriptCollisionGenerationMethod Enum=ConvexHulls
     bAutoDetectSpheres:bool=True
     bAutoDetectBoxes:bool=True
     bAutoDetectCapsules:bool=True
     MinThickness:float=1
     bSimplifyHulls:bool=True
     ConvexHullTargetFaceCount:int=25
     ConvexHullGeometricTolerance:float=0
     MaxConvexHullsPerMesh:int=1
     ConvexDecompositionSearchFactor:float=0.5
     ConvexDecompositionErrorTolerance:float=0
     ConvexDecompositionMinPartThickness:float=0.1
     SweptHullSimplifyTolerance:float=0.1
     SweptHullAxis:EGeometryScriptSweptHullAxis Enum=Z
     bRemoveFullyContainedShapes:bool=True
     MaxShapeCount:int=0
    ) -> (GeometryScriptCollisionFromMeshOptions:Geometry Script Collision From Mesh Options Structure) {is=title K2Node_MakeStruct}
    StaticMeshCollisionOptions:Geometry Script Set Static Mesh Collision Options Structure=(bMarkAsCustomized=True)
   ) -> (then, ReturnValue:Dynamic Mesh Object Reference)
  ]}
 SetTreeProperties {nodes [
   SetTreeProperties!() -> (then, LerpUVs:bool=true) {is=title K2Node_FunctionEntry}
   > n1 = GetSelectedAssets()
   > n2 = For Each Loop(Array=@n1)
   > n3 = Cast To StaticMesh(Object=@n2)
   > Set Mesh(Mesh:Static Mesh Object Reference=@n3) -> (Output_Get:Static Mesh Object Reference)
   > n4 = Sequence!(execute) -> (then_0, then_1, then_2, then_3:Exec, then_4:Exec, then_5:Exec)
   n5 = Get Mesh!() -> (Mesh:Static Mesh Object Reference)
   n6 = Get LODForCollision!(self:Static Mesh Object Reference=@n5) -> (LODForCollision:int=0)
   > n7 = Branch(Condition="Equal (Integer)"(A=@n6))
   n8 = GetNumLODs(self=@n5)
   > Set LODForCollision!(execute, LODForCollision:int=SelectInt(A=@n8, B=2, bPickA="integer < integer"(A=@n8, B=2)), self:Static Mesh Object Reference=@n5) -> (then, Output_Get:int=0)
   n9 = MakeLiteralString(Value="LOD collision = 2 \r\n")
   > PrintString(InString=@n9)
   > Set LOD Text(LOD Text:String=@n9) -> (Output_Get:String)
   n10 = Append(A=Append(A="LOD collision = ", B="To String (Integer)"(InInt=@n6)), B=" LOD has not been changed\r\n")
   PrintString(execute=@n7.else, InString=@n10)
   > Set LOD Text(LOD Text=@n10)
   n11 = Execute Python Script!(
    execute=@n4.then_1
    PythonScript="import unreal\r\n\r\nif 'static_mesh' in dir() and static_mesh and isinstance(static_mesh, unreal.StaticMesh):\r\n\r\n    ## Get the StaticMeshEditorSubsystem (this is the key for UE 5.7!)\r\n    editor_subsystem = unreal.get_editor_subsystem(unreal.StaticMeshEditorSubsystem)\r\n\r\n    ## Enable Nanite\r\n    try:\r\n        nanite_settings = editor_subsystem.get_nanite_settings(static_mesh)\r\n        nanite_settings.enabled = True\r\n        editor_subsystem.set_nanite_settings(static_mesh, nanite_settings)\r\n        unreal.log(f\"SUCCESS: Set Nanite to True for {static_mesh.get_name()}\")\r\n    except Exception as e:\r\n        unreal.log_error(f\"FAILED: Could not set Nanite: {str(e)}\")\r\n\r\n    ## Set Generate Two Sided Distance Field\r\n    try:\r\n        lod_index = 0\r\n        build_settings = editor_subsystem.get_lod_build_settings(static_mesh, lod_index)\r\n\r\n        ## Set Generate Two Sided Distance Field to True\r\n        build_settings.generate_distance_field_as_if_two_sided = True\r\n\r\n        ## Apply the modified build settings\r\n        editor_subsystem.set_lod_build_settings(static_mesh, lod_index, build_settings)\r\n        unreal.log(f\"SUCCESS: Set Generate Two Sided Distance Field to True for {static_mesh.get_name()}\")\r\n\r\n    except Exception as e:\r\n        unreal.log_error(f\"FAILED: Could not set Two Sided Distance Field: {str(e)}\")\r\n\r\n    ## Set Distance Field Self Shadow Bias to 50 if currently 0\r\n    try:\r\n        current_bias = static_mesh.get_editor_property(\"DistanceFieldSelfShadowBias\")\r\n        if current_bias == 0.0:\r\n            static_mesh.set_editor_property(\"DistanceFieldSelfShadowBias\", 50.0)\r\n            unreal.log(f\"SUCCESS: Set Distance Field Self Shadow Bias to 50 for {static_mesh.get_name()} (was 0)\")\r\n        else:\r\n            unreal.log(f\"INFO: Distance Field Self Shadow Bias already set to {current_bias}, skipping\")\r\n    except Exception as e:\r\n        unreal.log_error(f\"FAILED: Could not set Distance Field Self Shadow Bias: {str(e)}\")\r\n\r\n    ## Save the asset\r\n    try:\r\n        unreal.EditorAssetLibrary.save_loaded_asset(static_mesh)\r\n        unreal.log(f\"SUCCESS: Saved changes to {static_mesh.get_name()}\")\r\n    except Exception as e:\r\n        unreal.log_error(f\"FAILED: Could not save: {str(e)}\")\r\n\r\nelse:\r\n    unreal.log_error(\"No valid Static Mesh provided. Make sure to connect a Static Mesh to the input.\")"
    static_mesh:Static Mesh Object Reference=Get Mesh()
   ) -> (then, ReturnValue) {inputs [static_mesh]}
   > Branch(Condition=@n11)
   n12 = Execute Python Script(
    execute=@n4.then_2
    PythonScript="import unreal\r\n\r\n## Initialize return value\r\nsetting_was_changed = False\r\n\r\nif 'static_mesh' in dir() and static_mesh and isinstance(static_mesh, unreal.StaticMesh):\r\n\r\n    ## Get the StaticMeshEditorSubsystem\r\n    editor_subsystem = unreal.get_editor_subsystem(unreal.StaticMeshEditorSubsystem)\r\n\r\n    ## Set Nanite Shape Preservation to Preserve Area\r\n    try:\r\n        nanite_settings = editor_subsystem.get_nanite_settings(static_mesh)\r\n\r\n        previous_value = None\r\n        new_value = None\r\n        property_set_successfully = False\r\n\r\n        ## Try different possible property names for shape preservation\r\n        ## The diagnostic will tell us which one is correct\r\n\r\n        ## Attempt 1: Check if there's a shape_preservation property\r\n        if hasattr(nanite_settings, 'shape_preservation'):\r\n            ## Store the previous value\r\n            previous_value = nanite_settings.shape_preservation\r\n\r\n            ## Assuming it's an enum, try to set it to preserve area\r\n            ## Common enum values might be: PRESERVE_AREA, PreserveArea, etc.\r\n            try:\r\n                nanite_settings.shape_preservation = unreal.NaniteShapePreservation.PRESERVE_AREA\r\n                new_value = nanite_settings.shape_preservation\r\n                property_set_successfully = True\r\n                unreal.log(f\"Set Shape Preservation using NaniteShapePreservation.PRESERVE_AREA\")\r\n            except:\r\n                try:\r\n                    ## Try as integer (enums can sometimes be set as int)\r\n                    nanite_settings.shape_preservation = 1  ## Typically 0=None, 1=PreserveArea\r\n                    new_value = nanite_settings.shape_preservation\r\n                    property_set_successfully = True\r\n                    unreal.log(f\"Set Shape Preservation using integer value\")\r\n                except:\r\n                    unreal.log_error(\"Could not set shape_preservation property\")\r\n\r\n        ## Attempt 2: Check for preserve_area boolean\r\n        elif hasattr(nanite_settings, 'preserve_area'):\r\n            previous_value = nanite_settings.preserve_area\r\n            nanite_settings.preserve_area = True\r\n            new_value = nanite_settings.preserve_area\r\n            property_set_successfully = True\r\n            unreal.log(f\"Set preserve_area to True\")\r\n\r\n        else:\r\n            unreal.log_error(\"Could not find shape preservation property - run diagnostic first!\")\r\n\r\n        if property_set_successfully:\r\n            ## Check if value actually changed\r\n            if previous_value != new_value:\r\n                setting_was_changed = True\r\n                unreal.log(f\"Value CHANGED: {previous_value} -> {new_value}\")\r\n            else:\r\n                unreal.log(f\"Value UNCHANGED: Already set to {new_value}\")\r\n\r\n            ## Apply the modified nanite settings\r\n            editor_subsystem.set_nanite_settings(static_mesh, nanite_settings)\r\n            unreal.log(f\"SUCCESS: Applied Nanite settings to {static_mesh.get_name()}\")\r\n\r\n            ## Save the asset only if something changed\r\n            if setting_was_changed:\r\n                try:\r\n                    unreal.EditorAssetLibrary.save_loaded_asset(static_mesh)\r\n                    unreal.log(f\"SUCCESS: Saved changes to {static_mesh.get_name()}\")\r\n                except Exception as e:\r\n                    unreal.log_error(f\"FAILED: Could not save: {str(e)}\")\r\n                    setting_was_changed = False  ## Mark as failed if save didn't work\r\n            else:\r\n                unreal.log(f\"SKIPPED SAVE: No changes to save for {static_mesh.get_name()}\")\r\n\r\n    except Exception as e:\r\n        unreal.log_error(f\"FAILED: Could not set Nanite Shape Preservation: {str(e)}\")\r\n        setting_was_changed = False\r\n\r\nelse:\r\n    unreal.log_error(\"No valid Static Mesh provided. Make sure to connect a Static Mesh to the input.\")\r\n    setting_was_changed = False\r\n\r\n## Return the result\r\nreturn_value = setting_was_changed\r\nunreal.log(f\"RETURN VALUE: {return_value}\")"
    static_mesh=Get Mesh()
   ) {inputs [static_mesh]}
   > n13 = Branch(Condition=@n12)
   > Set PreserveAreaText(PreserveAreaText:String="Shape Preservation set to \"Preserve Area\"\r\n") -> (Output_Get:String)
   Set PreserveAreaText(execute=@n13.else, PreserveAreaText="Shape Preservation is Already Correct\r\n")
   n14 = Get LerpUVs() -> (LerpUVs:bool)
   PrintString(execute=@n4.then_3, InString=Append(A="LERPUVS = ", B="To String (Boolean)"(InBool=@n14)))
   > n15 = Execute Python Script!(
    execute
    PythonScript="## Copy and paste this script into the Execute Python Script node\r\n## For Unreal Engine 5.7\r\n## Sets Nanite Lerp UVs to the specified value (True or False)\r\n## Make sure to connect a Static Mesh asset to the 'static_mesh' input pin\r\n## Make sure to connect a Boolean to the 'lerp_uvs_enabled' input pin\r\n## Returns: True if setting was changed, False if already set or failed\r\n\r\nimport unreal\r\n\r\n## Initialize return value\r\nsetting_was_changed = False\r\n\r\n## Check if lerp_uvs_enabled input is provided\r\nif 'lerp_uvs_enabled' not in dir():\r\n    unreal.log_error(\"No 'lerp_uvs_enabled' input provided. Connect a Boolean to the input pin.\")\r\n    setting_was_changed = False\r\n    return_value = setting_was_changed\r\nelif 'static_mesh' in dir() and static_mesh and isinstance(static_mesh, unreal.StaticMesh):\r\n\r\n    ## Log the input value\r\n    unreal.log(f\"Input: lerp_uvs_enabled = {lerp_uvs_enabled}\")\r\n\r\n    ## Get the StaticMeshEditorSubsystem\r\n    editor_subsystem = unreal.get_editor_subsystem(unreal.StaticMeshEditorSubsystem)\r\n\r\n    ## Set Nanite Lerp UVs setting\r\n    try:\r\n        nanite_settings = editor_subsystem.get_nanite_settings(static_mesh)\r\n\r\n        previous_value = None\r\n        new_value = None\r\n        property_set_successfully = False\r\n\r\n        ## Try different possible property names for Lerp UVs\r\n        ## Based on diagnostic: the actual property is 'lerp_u_vs' (with underscores between each letter)\r\n\r\n        ## Attempt 1: Check if there's a lerp_u_vs property (the actual property name from UE)\r\n        if hasattr(nanite_settings, 'lerp_u_vs'):\r\n            ## Store the previous value\r\n            previous_value = nanite_settings.lerp_u_vs\r\n\r\n            ## Set the new value\r\n            nanite_settings.lerp_u_vs = lerp_uvs_enabled\r\n            new_value = nanite_settings.lerp_u_vs\r\n            property_set_successfully = True\r\n            unreal.log(f\"Set lerp_u_vs to {lerp_uvs_enabled}\")\r\n\r\n        ## Attempt 2: Check for lerp_uvs (alternative spelling)\r\n        elif hasattr(nanite_settings, 'lerp_uvs'):\r\n            previous_value = nanite_settings.lerp_uvs\r\n            nanite_settings.lerp_uvs = lerp_uvs_enabled\r\n            new_value = nanite_settings.lerp_uvs\r\n            property_set_successfully = True\r\n            unreal.log(f\"Set lerp_uvs to {lerp_uvs_enabled}\")\r\n\r\n        ## Attempt 3: Check for lerp_uv (singular)\r\n        elif hasattr(nanite_settings, 'lerp_uv'):\r\n            previous_value = nanite_settings.lerp_uv\r\n            nanite_settings.lerp_uv = lerp_uvs_enabled\r\n            new_value = nanite_settings.lerp_uv\r\n            property_set_successfully = True\r\n            unreal.log(f\"Set lerp_uv to {lerp_uvs_enabled}\")\r\n\r\n        else:\r\n            unreal.log_error(\"Could not find Lerp UVs property - run diagnostic first!\")\r\n            ## List all available properties for debugging\r\n            all_attrs = [a for a in dir(nanite_settings) if not a.startswith('_')]\r\n            unreal.log_warning(f\"Available properties: {', '.join(sorted(all_attrs))}\")\r\n\r\n        if property_set_successfully:\r\n            ## Check if value actually changed\r\n            if previous_value != new_value:\r\n                setting_was_changed = True\r\n                unreal.log(f\"Value CHANGED: {previous_value} -> {new_value}\")\r\n            else:\r\n                unreal.log(f\"Value UNCHANGED: Already set to {new_value}\")\r\n\r\n            ## Apply the modified nanite settings\r\n            editor_subsystem.set_nanite_settings(static_mesh, nanite_settings)\r\n            unreal.log(f\"SUCCESS: Applied Nanite settings to {static_mesh.get_name()}\")\r\n\r\n            ## Save the asset only if something changed\r\n            if setting_was_changed:\r\n                try:\r\n                    unreal.EditorAssetLibrary.save_loaded_asset(static_mesh)\r\n                    unreal.log(f\"SUCCESS: Saved changes to {static_mesh.get_name()}\")\r\n                except Exception as e:\r\n                    unreal.log_error(f\"FAILED: Could not save: {str(e)}\")\r\n                    setting_was_changed = False  ## Mark as failed if save didn't work\r\n            else:\r\n                unreal.log(f\"SKIPPED SAVE: No changes to save for {static_mesh.get_name()}\")\r\n\r\n    except Exception as e:\r\n        unreal.log_error(f\"FAILED: Could not set Nanite Lerp UVs: {str(e)}\")\r\n        setting_was_changed = False\r\n\r\nelse:\r\n    unreal.log_error(\"No valid Static Mesh provided. Make sure to connect a Static Mesh to the input.\")\r\n    setting_was_changed = False\r\n\r\n## Return the result\r\nreturn_value = setting_was_changed\r\nunreal.log(f\"RETURN VALUE: {return_value}\")\r\n"
    static_mesh=Get Mesh()
    lerp_uvs_enabled:bool=@n14
   ) -> (then, ReturnValue) {inputs [static_mesh, lerp_uvs_enabled]}
   > n16 = Branch(Condition=@n15)
   > Set LerpUVsText(LerpUVsText:String="LerpUVs has been CHANGED\r\n") -> (Output_Get:String)
   Set LerpUVsText(execute=@n16.else, LerpUVsText="LerpUVs has NOT been changed\r\n")
   n17 = Append(
    A=Append(A=MakeLiteralString(Value="I have set the follow tree properties \r\n\r\n"), B=Get LOD Text() -> (LOD Text:String))
    B=Get PreserveAreaText() -> (PreserveAreaText:String)
   )
   n18 = Get LerpUVsText() -> (LerpUVsText:String)
   Show Message Dialog(
    execute=@n4.then_5
    Title=MakeLiteralText(Value="LOCTEXT(\"Tree Properties\")")
    Message="To Text (String)"(InString=Append(A=@n17, B=@n18))
    MessageType=Ok
    DefaultValue=No
    MessageCategory=Warning
   )
   "Execute Python Command (Advanced)"!(
    execute
    PythonCommand:String="import unreal\r\n\r\n@unreal.uclass()\r\nclass MeshSettingsLibrary(unreal.BlueprintFunctionLibrary):\r\n\r\n    @unreal.ufunction(static=True, params=[unreal.StaticMesh], ret=bool, meta=dict(Category=\"Mesh Settings\"))\r\n    def set_mesh_settings(static_mesh):\r\n        \"\"\"\r\n        Configure static mesh settings:\r\n        - Enable Generate Two Sided Distance Field\r\n        - Enable Nanite\r\n        - Set Distance Field Self Shadow Bias to 50 (only if currently 0)\r\n\r\n        Args:\r\n            static_mesh: The Static Mesh asset to modify\r\n\r\n        Returns:\r\n            bool: True if successful, False otherwise\r\n        \"\"\"\r\n        if not static_mesh or not isinstance(static_mesh, unreal.StaticMesh):\r\n            unreal.log_error(\"Invalid Static Mesh provided\")\r\n            return False\r\n\r\n        try:\r\n            ## Get the current build settings\r\n            build_settings = static_mesh.get_editor_property(\"build_settings\")\r\n\r\n            ## Set Generate Two Sided Distance Field to True\r\n            build_settings.set_editor_property(\"generate_distance_field_as_if_two_sided\", True)\r\n            unreal.log(f\"Set Generate Two Sided Distance Field to True for {static_mesh.get_name()}\")\r\n\r\n            ## Set Distance Field Self Shadow Bias to 50 if currently 0\r\n            current_bias = build_settings.get_editor_property(\"distance_field_self_shadow_bias\")\r\n            if current_bias == 0:\r\n                build_settings.set_editor_property(\"distance_field_self_shadow_bias\", 50.0)\r\n                unreal.log(f\"Set Distance Field Self Shadow Bias to 50 for {static_mesh.get_name()} (was 0)\")\r\n            else:\r\n                unreal.log(f\"Distance Field Self Shadow Bias already set to {current_bias} for {static_mesh.get_name()}, skipping\")\r\n\r\n            ## Apply the build settings\r\n            static_mesh.set_editor_property(\"build_settings\", build_settings)\r\n\r\n            ## Enable Nanite\r\n            nanite_settings = static_mesh.get_editor_property(\"nanite_settings\")\r\n            nanite_settings.set_editor_property(\"enabled\", True)\r\n            static_mesh.set_editor_property(\"nanite_settings\", nanite_settings)\r\n            unreal.log(f\"Set Nanite to True for {static_mesh.get_name()}\")\r\n\r\n            ## Mark the asset as modified\r\n            unreal.EditorAssetLibrary.save_loaded_asset(static_mesh)\r\n            unreal.log(f\"Saved changes to {static_mesh.get_name()}\")\r\n\r\n            return True\r\n\r\n        except Exception as e:\r\n            unreal.log_error(f\"Failed to set mesh settings: {str(e)}\")\r\n            return False"
    ExecutionMode:EPythonCommandExecutionMode Enum=ExecuteFile
    FileExecutionScope:EPythonFileExecutionScope Enum=Private
   ) -> (then, CommandResult:String, LogOutput:Array of Python Log Output Entry Structures, ReturnValue:bool)
   MakeLiteralText(Value="LOCTEXT(\"I have set the follow tree properties\")")
   Get DistanceFieldText() -> (DistanceFieldText:String)
  ]}
]
variables [Collision {type=StaticMesh, default=null}]
