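; Editor tree tool that merges separate trunk and canopy impostor meshes into one LOD of an existing static mesh asset.
; MergeMesh copies both into dynamic compute meshes via Geometry Script, painting only the trunk a constant cyan vertex color.
; It appends the canopy, remaps material ID 0 to the target index, then writes to the asset path if that asset exists.
; A target LOD of -1 falls back to LOD 3.
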(fn MergeMesh
  (params (TrunkMesh :type Static%20Mesh%20Object%20Reference) (CanopyMesh :type Static%20Mesh%20Object%20Reference) (Asset%20Path :type String) (TargetLODIndex :type Integer) (TargetMaterialIDIndex :type Integer :default 2))
  (Variables|Default|SetAsset%20Path Asset%20Path)
  (Variables|Default|SetTargetLODIndex TargetLODIndex)
  (Utilities|FlowControl|Sequence
    (:then_0
      (bind _returnvalue (DynamicMeshActor|AllocateComputeMesh))
      (bind _returnvalue_2 (GeometryScript|StaticMesh|CopyMeshFromStaticMesh TrunkMesh _returnvalue)
        (:Success
          (bind _output_get (LocalVariables|SetDyTrunk _returnvalue_2))
          (GeometryScript|VertexColor|SetMeshConstantVertexColor :TargetMesh _output_get :Color "(R=0.000000,G=1.000000,B=1.000000,A=1.000000)" :bClearExisting true))))
    (:then_1
      (bind _returnvalue_3 (DynamicMeshActor|AllocateComputeMesh))
      (bind _returnvalue_4 (GeometryScript|StaticMesh|CopyMeshFromStaticMesh CanopyMesh _returnvalue_3)
        (:Success
          (LocalVariables|SetDyCanopy _returnvalue_4))))
    (:then_2
      (bind _returnvalue_5 (GeometryScript|MeshEdits|AppendMesh :TargetMesh (LocalVariables|GetDyTrunk) :AppendMesh (LocalVariables|GetDyCanopy) :bDeferChangeNotifications true :AppendOptions (struct :CombineMode "UseSource")))
      (bind _returnvalue_6 (GeometryScript|Materials|RemapMaterialIDs _returnvalue_5 0 (Variables|GetTargetMaterialIDIndex)))
      (bind _returnvalue_7 (EditorScripting|Asset|DoesAssetExist (Variables|Default|GetAsset%20Path) :!class /Script/EditorScriptingUtilities.EditorAssetLibrary))
      (if _returnvalue_7
        (bind _returnvalue_8 (EditorScripting|Asset|LoadAsset (Variables|Default|GetAsset%20Path) :!class /Script/EditorScriptingUtilities.EditorAssetLibrary))
        (bind _asstatic_mesh (Utilities|Casting|CastToStaticMesh _returnvalue_8))
        (GeometryScript|StaticMesh|CopyMeshtoStaticMesh _returnvalue_6 _asstatic_mesh (Utilities|Struct|MakeGeometryScriptCopyMeshtoAssetOptions true true) (Utilities|Struct|MakeGeometryScriptMeshWriteLOD false (Math|Integer|SelectInt 3 (Variables|Default|GetTargetLODIndex) (== (Variables|Default|GetTargetLODIndex) -1))) true)))))

(fn ConstructionScript)
