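; Editor bake tool that converts Asteria trees from data-asset entries into baked meshes, impostors and attribute textures.
; Baking drives a GeoScript converter, infuses custom LODs, assigns materials by conversion mode (including texture array and UDIM), updates tree Blueprints and spawns comparison actors.
; Material parameters are written by index into a render target saved as a per-tree texture.
; Impostors come from a freshly spawned baker per tree, with custom and split trunk/canopy variants.
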(event Custom|SpawnClassEvent
  (params (Class :type Actor%20Class%20Reference) (SpawnTransform :type Transform))
  (CallFunction|SpawnClass Class SpawnTransform))

(event Custom|TestRT
  (Rendering|ClearRenderTarget2D 0 (Variables|Default|GetRenderTarget_2D))
  (bind _returnvalue (Rendering|Material|CreateDynamicMaterialInstance 0 "/Game/Environments/Asteria/Foliage/Trees/MasterMaterials/BakeArrayTexture/M_Attribute_Bake.M_Attribute_Bake" "None" "None" :!class /Script/Engine.KismetMaterialLibrary))
  (Rendering|Material|SetScalarParameterValue _returnvalue "Index" 16.0 :!class /Script/Engine.MaterialInstanceDynamic)
  (Rendering|Material|SetScalarParameterValue _returnvalue "Value" 1.0 :!class /Script/Engine.MaterialInstanceDynamic)
  (Rendering|Material|SetVectorParameterValue _returnvalue "Vector3" "(R=1.000000,G=0.004681,B=0.000000,A=1.000000)" :!class /Script/Engine.MaterialInstanceDynamic)
  (Rendering|DrawMaterialtoRenderTarget 0 (Variables|Default|GetRenderTarget_2D) _returnvalue))

(fn BakeAllImpostor
  (CallFunction|SetTreeDatafromDataAsset)
  (for (_ _array_index) (Variables|DontTouchOnlyRead|GetTreeData)
    (CallFunction|BakeImpostor _array_index)))

(fn BakeAllTrees
  (CallFunction|SetTreeDatafromDataAsset)
  (CallFunction|DeleteallSpawnedActors)
  (Utilities|FlowControl|Sequence
    (:then_0
      (for (_ _array_index) (Variables|DontTouchOnlyRead|GetTreeData)
        (CallFunction|CreateBakedMesh _array_index)
        (switch Utilities|FlowControl|Switch|SwitchonConvertTree_Mode (Variables|ConversionMode|GetMode)
          (:NewEnumerator1
            _)
          (:NewEnumerator2
            _))
        (CallFunction|CopyAllParametersandCreateAttributeTexture _array_index)))
    (:then_1
      (switch Utilities|FlowControl|Switch|SwitchonConvertTree_Mode (Variables|ConversionMode|GetMode)
        (:NewEnumerator1
          _)
        (:NewEnumerator2
          _))
      (4AdvancedFunctionsMostlyforTesting|SetTexturesfromMode))))

(fn BakeonlyAttributeTexture
  (CallFunction|CopyAllParametersandCreateAttributeTexture (Variables|2BakeSingleTree|GetTree%20Index%20To%20Bake)))

(fn BakeSingleImpostor
  (CallFunction|SetTreeDatafromDataAsset)
  (CallFunction|BakeImpostor (Variables|2BakeSingleTree|GetTree%20Index%20To%20Bake)))

(fn BakeSingleTree
  (CallFunction|SetTreeDatafromDataAsset)
  (CallFunction|DeleteallSpawnedActors)
  (Utilities|FlowControl|Sequence
    (:then_0
      (CallFunction|CreateBakedMesh (Variables|2BakeSingleTree|GetTree%20Index%20To%20Bake))
      (switch Utilities|FlowControl|Switch|SwitchonConvertTree_Mode (Variables|ConversionMode|GetMode)
        (:NewEnumerator1
          _)
        (:NewEnumerator2
          _))
      (CallFunction|CopyAllParametersandCreateAttributeTexture (Variables|2BakeSingleTree|GetTree%20Index%20To%20Bake)))
    (:then_1
      (switch Utilities|FlowControl|Switch|SwitchonConvertTree_Mode (Variables|ConversionMode|GetMode)
        (:NewEnumerator1
          _)
        (:NewEnumerator2
          _))
      (4AdvancedFunctionsMostlyforTesting|SetTexturesfromMode))))

(fn BakeImpostor
  (params (Index :type Integer))
  (bind _treedata (Variables|DontTouchOnlyRead|GetTreeData))
  (bind _output (Utilities|Array|Get(acopy) _treedata Index))
  (bind (_outputname_34_04cb671a4066ad0bd2f6678c3be2ba7a _staticmesh_tree_nanite_31_280c65834a5f8d3da1beea87a2fdfc78 _staticmesh_tree_fallback_30_b769c69848448d3b2441219dcb6571a6 _customlods_38_dc1c72d2449fa58507dc50a878a3d5ae _addcustomlodmaterials_41_9658b18e4610e22c9579b9b38269adf2 _blueprint_tree_15_72b637784d14b76a881869aee9983fa7 _trunk_material_element_9_86d6d8c240e634354ed6179ad68da4f5 _leaves_material_element_11_d8f23070444ef180f803ec866de8a069 _mesh_material_idformask_23_33f6f1dc43f19a49d55212a45a6b7562 _collisionmesh_27_d515c7ef4923549dd36621918a253b7d _usecustomimpostor_43_b063f0534c360f988d892797fb015e8e _customimpostormesh_46_e40a5445444cf88f0cf789aad2edf530 _flipmaterialidsfornaniteandfallbackmissmatch_48_5b1aa6cd4be3f574e847b290b81e8659 _usesplittrunkcanopyimpostor_52_c4828f9f4887afbaa415b896e3fbf2c8 _staticmesh_canopy_55_dbd0e7b3448eb112dbe752b8fe499d68 _staticmesh_trunk_56_f8f65d9b4c9be16c7cfa6da199f1fcb6) (Utilities|Struct|BreakTreetoConvertData _output))
  (bind _index (Variables|GetIndex))
  (bind _treedata_2 (Variables|DontTouchOnlyRead|GetTreeData))
  (bind _output_2 (Utilities|Array|Get(acopy) _treedata_2 _index))
  (bind (_outputname_34_04cb671a4066ad0bd2f6678c3be2ba7a_2 _staticmesh_tree_nanite_31_280c65834a5f8d3da1beea87a2fdfc78_2 _staticmesh_tree_fallback_30_b769c69848448d3b2441219dcb6571a6_2 _customlods_38_dc1c72d2449fa58507dc50a878a3d5ae_2) (Utilities|Struct|BreakTreetoConvertData _output_2))
  (bind _index_2 (Variables|GetIndex))
  (bind _tree_internaldata (Variables|Default|GetTree_internalData))
  (bind _output_3 (Utilities|Array|Get(acopy) _tree_internaldata _index_2))
  (bind (_trunkmaterial_2_25c5fc6d4dddf7977eaf2da64d33d2ed _leavesmaterial_4_2a592d9b44092b0bfe5d00b8e6546ef5 _assetbasefolder_7_39c096b443542bc75c0e6387361d3c9a _treename_9_53616c4448ee21eef59b3a9e0a39aabf) (Utilities|Struct|BreakTreeConverteInternalData _output_3))
  (bind _index_3 (Variables|GetIndex))
  (bind _tree_internaldata_2 (Variables|Default|GetTree_internalData))
  (bind _output_4 (Utilities|Array|Get(acopy) _tree_internaldata_2 _index_3))
  (bind (_trunkmaterial_2_25c5fc6d4dddf7977eaf2da64d33d2ed_2 _leavesmaterial_4_2a592d9b44092b0bfe5d00b8e6546ef5_2 _assetbasefolder_7_39c096b443542bc75c0e6387361d3c9a_2 _treename_9_53616c4448ee21eef59b3a9e0a39aabf_2) (Utilities|Struct|BreakTreeConverteInternalData _output_4))
  (bind _index_4 (Variables|GetIndex))
  (bind _tree_internaldata_3 (Variables|Default|GetTree_internalData))
  (bind _output_5 (Utilities|Array|Get(acopy) _tree_internaldata_3 _index_4))
  (bind (_trunkmaterial_2_25c5fc6d4dddf7977eaf2da64d33d2ed_3 _leavesmaterial_4_2a592d9b44092b0bfe5d00b8e6546ef5_3 _assetbasefolder_7_39c096b443542bc75c0e6387361d3c9a_3 _treename_9_53616c4448ee21eef59b3a9e0a39aabf_3) (Utilities|Struct|BreakTreeConverteInternalData _output_5))
  (bind _staticmaterials (Variables|/Script/Engine.StaticMesh|GetStaticMaterials _staticmesh_tree_fallback_30_b769c69848448d3b2441219dcb6571a6_2))
  (bind _returnvalue (Utilities|Array|Length _staticmaterials))
  (bind _staticmaterials_2 (Variables|/Script/Engine.StaticMesh|GetStaticMaterials _staticmesh_tree_fallback_30_b769c69848448d3b2441219dcb6571a6))
  (bind _returnvalue_2 (Utilities|Array|Length _staticmaterials_2))
  (bind _returnvalue_3 (Utilities|Array|Length _customlods_38_dc1c72d2449fa58507dc50a878a3d5ae))
  (bind _returnvalue_4 (> _returnvalue_3 0))
  (bind _returnvalue_5 (+ _returnvalue_3 1))
  (bind _returnvalue_6 (Math|Integer|SelectInt _returnvalue_5 3 _returnvalue_4))
  (bind _returnvalue_7 (Utilities|String|Append _assetbasefolder_7_39c096b443542bc75c0e6387361d3c9a_2 "BakeData/"))
  (bind _returnvalue_8 (Utilities|String|Append _treename_9_53616c4448ee21eef59b3a9e0a39aabf_2 "_Bake"))
  (bind _returnvalue_9 (Utilities|String|Append _returnvalue_7 _returnvalue_8))
  (bind _returnvalue_10 (Utilities|String|Append _assetbasefolder_7_39c096b443542bc75c0e6387361d3c9a_3 "BakeData/"))
  (bind _returnvalue_11 (Utilities|String|Append _treename_9_53616c4448ee21eef59b3a9e0a39aabf_3 "_Bake"))
  (bind _returnvalue_12 (Utilities|String|Append _returnvalue_10 _returnvalue_11))
  (bind _returnvalue_13 (Utilities|String|Append _assetbasefolder_7_39c096b443542bc75c0e6387361d3c9a "Impostor/"))
  (Utilities|FlowControl|Sequence
    (:then_0
      (if _usesplittrunkcanopyimpostor_52_c4828f9f4887afbaa415b896e3fbf2c8
        ;[1 1 1] This section is dedicated to merge the canopy impostor mesh to the trunk static mesh.
        (Utilities|FlowControl|Sequence
          (:then_0
            _)
          (:then_1
            (bind _returnvalue_14 (AddComponent|NODE_AddChildActorComponent-4 :RelativeTransform (Math|Transform|MakeTransform "-8000.000000,3000.000000, 0")))
            (bind _asbp_merge_split_impostor (Utilities|Casting|CastToBP_MergeSplitImpostor (Variables|/Script/Engine.ChildActorComponent|GetChildActor _returnvalue_14)))
            (bind _returnvalue_15 (EditorScripting|Asset|LoadAsset (Utilities|String|Append _returnvalue_13 (Utilities|String|Append _treename_9_53616c4448ee21eef59b3a9e0a39aabf "_Impostor")) :!class /Script/EditorScriptingUtilities.EditorAssetLibrary))
            (bind _asstatic_mesh (Utilities|Casting|CastToStaticMesh _returnvalue_15))
            (Class|BPMergeSplitImpostor|MergeMesh _asbp_merge_split_impostor _staticmesh_trunk_56_f8f65d9b4c9be16c7cfa6da199f1fcb6 _asstatic_mesh (Utilities|String|Append (Utilities|String|Append _assetbasefolder_7_39c096b443542bc75c0e6387361d3c9a "BakeData/") (Utilities|String|Append _treename_9_53616c4448ee21eef59b3a9e0a39aabf "_Bake")) _returnvalue_6 _returnvalue_2)
            (bind _returnvalue_16 (EditorScripting|Asset|LoadAsset (Utilities|String|Append _returnvalue_13 (Utilities|String|Append (Utilities|String|Append "MI_" _treename_9_53616c4448ee21eef59b3a9e0a39aabf) "_Impostor")) :!class /Script/EditorScriptingUtilities.EditorAssetLibrary))
            (bind _asmaterial_instance_constant (Utilities|Casting|CastToMaterialInstanceConstant _returnvalue_16))
            (MaterialEditing|SetMaterialInstanceStaticSwitchParameterValue _asmaterial_instance_constant "Use_StaticTrunk" true))
          (:then_2
            (Components|DestroyComponent _returnvalue_14 0)))
        (elif _usecustomimpostor_43_b063f0534c360f988d892797fb015e8e
          ;[1 1 1] This section is to add a custom impostor instead of generate a new one.
          (Utilities|FlowControl|Sequence
            (:then_0
              (bind _returnvalue_17 (AddComponent|NODE_AddChildActorComponent-3 :RelativeTransform (Math|Transform|MakeTransform "-8000.000000,3000.000000, 0")))
              (bind _asbp_add_custom_impostor (Utilities|Casting|CastToBP_AddCustomImpostor (Variables|/Script/Engine.ChildActorComponent|GetChildActor _returnvalue_17)))
              (Variables|/Game/ImpostorBaker/BP/BP_AddCustomImpostor.BP_AddCustomImpostor_C|SetTargetLODIndex _returnvalue_6 _asbp_add_custom_impostor)
              (Variables|/Game/ImpostorBaker/BP/BP_AddCustomImpostor.BP_AddCustomImpostor_C|SetStaticmesh _customimpostormesh_46_e40a5445444cf88f0cf789aad2edf530 _asbp_add_custom_impostor)
              (Variables|/Game/ImpostorBaker/BP/BP_AddCustomImpostor.BP_AddCustomImpostor_C|SetTargetMaterialID _returnvalue_2 _asbp_add_custom_impostor)
              (Variables|/Game/ImpostorBaker/BP/BP_AddCustomImpostor.BP_AddCustomImpostor_C|SetExistingMeshAssetName _returnvalue_12 _asbp_add_custom_impostor)
              (Class|BPAddCustomImpostor|3%29CreateStaticAssets _asbp_add_custom_impostor)
              (bind _returnvalue_18 (EditorScripting|Asset|LoadAsset _returnvalue_12 :!class /Script/EditorScriptingUtilities.EditorAssetLibrary))
              (bind _asstatic_mesh_2 (Utilities|Casting|CastToStaticMesh _returnvalue_18))
              (LocalVariables|SetStaticMeshWithImpostor _asstatic_mesh_2))
            (:then_1
              (Components|DestroyComponent _returnvalue_17 0)))))
      (Utilities|FlowControl|Sequence
        (:then_0
          ;[1 1 1] Spawn Static Mesh actor and assign Material with the Impostor switch set that makes impostor baking possible
          (bind _returnvalue_19 (Game|SpawnActorfromClass 0 "/Script/Engine.StaticMeshActor" (Math|Transform|MakeTransform "0, 0, 0") "Undefined" "MultiplyWithRoot" 0 0))
          (bind _output_get (LocalVariables|SetTempStaticMeshActor _returnvalue_19))
          (Components|StaticMesh|SetStaticMesh (Variables|/Script/Engine.StaticMeshActor|GetStaticMeshComponent _output_get) (select _usesplittrunkcanopyimpostor_52_c4828f9f4887afbaa415b896e3fbf2c8 _staticmesh_canopy_55_dbd0e7b3448eb112dbe752b8fe499d68 (select (Variables|Debug|GetUse%20NaniteMesh%20for%20Impostor%20Bake) _staticmesh_tree_nanite_31_280c65834a5f8d3da1beea87a2fdfc78 _staticmesh_tree_fallback_30_b769c69848448d3b2441219dcb6571a6)))
          (Variables|/Script/Engine.Actor|SetTags (Utilities|Array|MakeArray "AutoSpawned") _output_get))
        (:then_1
          ;[1 1 1] Spawn new Impostor baker every loop and destroy. there was a bug when using the same Component each time
          (bind _returnvalue_20 (AddComponent|NODE_AddChildActorComponent-2 :RelativeTransform (Math|Transform|MakeTransform "-8000.000000,3000.000000, 0")))
          (bind _asbp_generate_imposter_sprites_2_fnasteria (Utilities|Casting|CastToBP_Generate_ImposterSprites_2_FNAsteria (Variables|/Script/Engine.ChildActorComponent|GetChildActor _returnvalue_20)))
          (LocalVariables|SetCurrentImpostorBaker _asbp_generate_imposter_sprites_2_fnasteria)
          (Utilities|FlowControl|Sequence
            (:then_0
              (bind _impostormaterialoverride (Variables|Debug|GetImpostorMaterialOverride)
                (:then
                  (Variables|/Game/ImpostorBaker/BP/BP_Generate_ImposterSprites_2_FNAsteria.BP_Generate_ImposterSprites_2_FNAsteria_C|SetPreview%20Parent%20Material _impostormaterialoverride (LocalVariables|GetCurrentImpostorBaker)))
                (:else)))
            (:then_1
              ;[1 1 1] Set Data for Output paths
              (Variables|/Game/ImpostorBaker/BP/BP_Generate_ImposterSprites_2_FNAsteria.BP_Generate_ImposterSprites_2_FNAsteria_C|SetStatic%20Mesh%20Actor (LocalVariables|GetTempStaticMeshActor) (LocalVariables|GetCurrentImpostorBaker))
              (Variables|/Game/ImpostorBaker/BP/BP_Generate_ImposterSprites_2_FNAsteria.BP_Generate_ImposterSprites_2_FNAsteria_C|SetTargetLODIndex (Math|Integer|SelectInt (+ (Utilities|Array|Length _customlods_38_dc1c72d2449fa58507dc50a878a3d5ae_2) 1) -1 (> (Utilities|Array|Length _customlods_38_dc1c72d2449fa58507dc50a878a3d5ae_2) 0)) (LocalVariables|GetCurrentImpostorBaker))
              (Variables|/Game/ImpostorBaker/BP/BP_Generate_ImposterSprites_2_FNAsteria.BP_Generate_ImposterSprites_2_FNAsteria_C|SetTargetMaterialIDIndex _returnvalue (LocalVariables|GetCurrentImpostorBaker))
              (Variables|/Game/ImpostorBaker/BP/BP_Generate_ImposterSprites_2_FNAsteria.BP_Generate_ImposterSprites_2_FNAsteria_C|SetNew%20Asset%20Path (Utilities|String|Append _assetbasefolder_7_39c096b443542bc75c0e6387361d3c9a_2 "Impostor/") (LocalVariables|GetCurrentImpostorBaker))
              (Variables|/Game/ImpostorBaker/BP/BP_Generate_ImposterSprites_2_FNAsteria.BP_Generate_ImposterSprites_2_FNAsteria_C|SetNew%20Asset%20Texture%20Prefix (Utilities|String|Append (Utilities|String|Append "T_" _treename_9_53616c4448ee21eef59b3a9e0a39aabf_2) "_Impostor") (LocalVariables|GetCurrentImpostorBaker))
              (Variables|/Game/ImpostorBaker/BP/BP_Generate_ImposterSprites_2_FNAsteria.BP_Generate_ImposterSprites_2_FNAsteria_C|SetNew%20Asset%20MIC%20Name (Utilities|String|Append (Utilities|String|Append "MI_" _treename_9_53616c4448ee21eef59b3a9e0a39aabf_2) "_Impostor") (LocalVariables|GetCurrentImpostorBaker))
              (Variables|/Game/ImpostorBaker/BP/BP_Generate_ImposterSprites_2_FNAsteria.BP_Generate_ImposterSprites_2_FNAsteria_C|SetNew%20Asset%20ImpostorMeshName (Utilities|String|Append _treename_9_53616c4448ee21eef59b3a9e0a39aabf_2 "_Impostor") (LocalVariables|GetCurrentImpostorBaker))
              (Variables|/Game/ImpostorBaker/BP/BP_Generate_ImposterSprites_2_FNAsteria.BP_Generate_ImposterSprites_2_FNAsteria_C|SetExistingMeshAssetName _returnvalue_9 (LocalVariables|GetCurrentImpostorBaker))
              ;[1 1 1] Trigger Impostor bake
              (Class|BPGenerateImposterSprites2FNAsteria|SetupAll (LocalVariables|GetCurrentImpostorBaker))
              (Class|BPGenerateImposterSprites2FNAsteria|1%29ClearRTs (LocalVariables|GetCurrentImpostorBaker))
              (Class|BPGenerateImposterSprites2FNAsteria|2%29RenderFrames (LocalVariables|GetCurrentImpostorBaker))
              (Class|BPGenerateImposterSprites2FNAsteria|3%29CreateStaticAssets (LocalVariables|GetCurrentImpostorBaker))
              ;[1 1 1] Set material on Impostor LOD and destroy Impostor baker for next loop
              (bind _returnvalue_21 (EditorScripting|Asset|LoadAsset _returnvalue_9 :!class /Script/EditorScriptingUtilities.EditorAssetLibrary))
              (bind _asstatic_mesh_3 (Utilities|Casting|CastToStaticMesh _returnvalue_21))
              (LocalVariables|SetStaticMeshWithImpostor _asstatic_mesh_3)
              (StaticMesh|SetMaterial _asstatic_mesh_3 (+ _returnvalue 0) (Variables|/Game/ImpostorBaker/BP/BP_Generate_ImposterSprites_2_FNAsteria.BP_Generate_ImposterSprites_2_FNAsteria_C|GetCreated%20MIC (LocalVariables|GetCurrentImpostorBaker)))
              (Actor|DestroyActor (LocalVariables|GetCurrentImpostorBaker)))))
        (:then_2
          (Actor|DestroyActor (LocalVariables|GetTempStaticMeshActor)))))
    (:then_1
      ;[1 1 1] Spawn Meshes just for visual comparison
      (bind _returnvalue_22 (Game|SpawnActorfromClass 0 "/Script/Engine.StaticMeshActor" (Math|Transform|MakeTransform (Math|Vector|MakeVector (* (Variables|Default|GetAssetSpacing) (Variables|GetIndex)) -5e+03)) "Undefined" "MultiplyWithRoot" 0 0))
      (Components|StaticMesh|SetStaticMesh (Variables|/Script/Engine.StaticMeshActor|GetStaticMeshComponent _returnvalue_22) (LocalVariables|GetStaticMeshWithImpostor))
      (Variables|/Script/Engine.Actor|SetTags (Utilities|Array|MakeArray "AutoSpawned") _returnvalue_22)
      (Rendering|LOD|SetForcedLodModel (Variables|/Script/Engine.StaticMeshActor|GetStaticMeshComponent _returnvalue_22) 10)
      (Variables|/Script/Engine.StaticMeshComponent|SetbDisallowNanite true (Variables|/Script/Engine.StaticMeshActor|GetStaticMeshComponent _returnvalue_22)))))

(fn CopyAllParametersandCreateAttributeTexture
  (params (TreeDataIndex :type Integer))
  (bind _treedataindex (Variables|GetTreeDataIndex))
  (bind _tree_internaldata (Variables|Default|GetTree_internalData))
  (bind _output (Utilities|Array|Get(acopy) _tree_internaldata _treedataindex))
  (bind (_trunkmaterial_2_25c5fc6d4dddf7977eaf2da64d33d2ed _leavesmaterial_4_2a592d9b44092b0bfe5d00b8e6546ef5) (Utilities|Struct|BreakTreeConverteInternalData _output))
  (bind _treedataindex_2 (Variables|GetTreeDataIndex))
  (bind _tree_internaldata_2 (Variables|Default|GetTree_internalData))
  (bind _output_2 (Utilities|Array|Get(acopy) _tree_internaldata_2 _treedataindex_2))
  (bind (_trunkmaterial_2_25c5fc6d4dddf7977eaf2da64d33d2ed_2 _leavesmaterial_4_2a592d9b44092b0bfe5d00b8e6546ef5_2 _assetbasefolder_7_39c096b443542bc75c0e6387361d3c9a _treename_9_53616c4448ee21eef59b3a9e0a39aabf) (Utilities|Struct|BreakTreeConverteInternalData _output_2))
  (bind _returnvalue (Utilities|String|Append _treename_9_53616c4448ee21eef59b3a9e0a39aabf "_Attributes"))
  (bind _returnvalue_2 (Utilities|String|Append _assetbasefolder_7_39c096b443542bc75c0e6387361d3c9a "BakeData/"))
  (bind _returnvalue_3 (Utilities|String|Append _returnvalue_2 _returnvalue))
  (bind _parameterindexlist (Variables|DontTouchOnlyRead|GetParameterIndexList))
  (Utilities|FlowControl|Sequence
    (:then_0
      (Rendering|ClearRenderTarget2D 0 (Variables|Default|GetRenderTarget_2D))
      (Rendering|ClearRenderTarget2D 0 (Variables|Default|GetRenderTarget_2D_TempCopy))
      ;[1 1 1] Write values to texture
      (for _array_element (Variables|Default|GetParameterDataList)
        (if (.isFloat3_9_D6570DD74B902E375681418D1F7E3CA6 (Utilities|Struct|BreakTreeParameterData _array_element))
          (bind _returnvalue_4 (Rendering|Material|GetVectorParameterValue (select (Utilities|Struct|BreakTreeParameterData _array_element) _trunkmaterial_2_25c5fc6d4dddf7977eaf2da64d33d2ed _leavesmaterial_4_2a592d9b44092b0bfe5d00b8e6546ef5) (Utilities|String|StringToName (.ParameterName_4_9C769BB24C9D282C35C078B0E7DB6B92 (Utilities|Struct|BreakTreeParameterData _array_element))) :!class /Script/Engine.MaterialInstanceDynamic))
          (Default|WriteVectorValuetoRendertarget (Math|Conversions|ToFloat(Integer) (Utilities|Array|FindItem _parameterindexlist (Utilities|String|StringToName (.ParameterName_4_9C769BB24C9D282C35C078B0E7DB6B92 (Utilities|Struct|BreakTreeParameterData _array_element))))) _returnvalue_4)
          (else
            (bind _returnvalue_5 (Rendering|Material|GetScalarParameterValue (select (Utilities|Struct|BreakTreeParameterData _array_element) _trunkmaterial_2_25c5fc6d4dddf7977eaf2da64d33d2ed _leavesmaterial_4_2a592d9b44092b0bfe5d00b8e6546ef5) (Utilities|String|StringToName (.ParameterName_4_9C769BB24C9D282C35C078B0E7DB6B92 (Utilities|Struct|BreakTreeParameterData _array_element))) :!class /Script/Engine.MaterialInstanceDynamic))
            (CallFunction|WriteScalarValuetoRendertarget (Math|Conversions|ToFloat(Integer) (Utilities|Array|FindItem _parameterindexlist (Utilities|String|StringToName (.ParameterName_4_9C769BB24C9D282C35C078B0E7DB6B92 (Utilities|Struct|BreakTreeParameterData _array_element))))) _returnvalue_5)
            (Development|PrintString 0)))))
    (:then_1
      ;[1 1 1] Save Render Target
      (bind _returnvalue_6 (EditorScripting|Asset|DoAssetsExist (Utilities|Array|MakeArray _returnvalue_3) :!class /Script/EditorScriptingUtilities.EditorAssetLibrary))
      (if _returnvalue_6
        (bind _returnvalue_7 (EditorScripting|Asset|LoadAsset _returnvalue_3 :!class /Script/EditorScriptingUtilities.EditorAssetLibrary))
        (bind _astexture_2d (Utilities|Casting|CastToTexture2D _returnvalue_7))
        (EditorScripting|Rendering|ConvertRenderTargettoTexture2DEditorOnly 0 (Variables|Default|GetRenderTarget_2D) _astexture_2d)
        (else
          (EditorScripting|Rendering|RenderTarget2DCreateStaticTexture2DEditorOnly (Variables|Default|GetRenderTarget_2D) _returnvalue_3 "TC_HDR" "TMGS_NoMipmaps"))))))

(fn CreateBakedMesh
  (params (AssetIndex :type Integer))
  (bind _currenttreedata (LocalVariables|GetCurrentTreeData))
  (bind (_outputname_34_04cb671a4066ad0bd2f6678c3be2ba7a _staticmesh_tree_nanite_31_280c65834a5f8d3da1beea87a2fdfc78 _staticmesh_tree_fallback_30_b769c69848448d3b2441219dcb6571a6 _customlods_38_dc1c72d2449fa58507dc50a878a3d5ae _addcustomlodmaterials_41_9658b18e4610e22c9579b9b38269adf2 _blueprint_tree_15_72b637784d14b76a881869aee9983fa7) (Utilities|Struct|BreakTreetoConvertData _currenttreedata))
  (bind _currenttreedata_2 (LocalVariables|GetCurrentTreeData))
  (bind (_outputname_34_04cb671a4066ad0bd2f6678c3be2ba7a_2 _staticmesh_tree_nanite_31_280c65834a5f8d3da1beea87a2fdfc78_2 _staticmesh_tree_fallback_30_b769c69848448d3b2441219dcb6571a6_2 _customlods_38_dc1c72d2449fa58507dc50a878a3d5ae_2 _addcustomlodmaterials_41_9658b18e4610e22c9579b9b38269adf2_2 _blueprint_tree_15_72b637784d14b76a881869aee9983fa7_2 _trunk_material_element_9_86d6d8c240e634354ed6179ad68da4f5 _leaves_material_element_11_d8f23070444ef180f803ec866de8a069) (Utilities|Struct|BreakTreetoConvertData _currenttreedata_2))
  (bind _assetindex (Variables|GetAssetIndex))
  (bind _tree_internaldata (Variables|Default|GetTree_internalData))
  (bind _output (Utilities|Array|Get(acopy) _tree_internaldata _assetindex))
  (bind (_trunkmaterial_2_25c5fc6d4dddf7977eaf2da64d33d2ed _leavesmaterial_4_2a592d9b44092b0bfe5d00b8e6546ef5 _assetbasefolder_7_39c096b443542bc75c0e6387361d3c9a _treename_9_53616c4448ee21eef59b3a9e0a39aabf) (Utilities|Struct|BreakTreeConverteInternalData _output))
  (bind (_trunkmaterial_2_25c5fc6d4dddf7977eaf2da64d33d2ed_2 _leavesmaterial_4_2a592d9b44092b0bfe5d00b8e6546ef5_2 _assetbasefolder_7_39c096b443542bc75c0e6387361d3c9a_2 _treename_9_53616c4448ee21eef59b3a9e0a39aabf_2) (Utilities|Struct|BreakTreeConverteInternalData (LocalVariables|GetCurrentInternalData)))
  (bind _currenttreedata_3 (LocalVariables|GetCurrentTreeData))
  (bind (_outputname_34_04cb671a4066ad0bd2f6678c3be2ba7a_3 _staticmesh_tree_nanite_31_280c65834a5f8d3da1beea87a2fdfc78_3 _staticmesh_tree_fallback_30_b769c69848448d3b2441219dcb6571a6_3 _customlods_38_dc1c72d2449fa58507dc50a878a3d5ae_3 _addcustomlodmaterials_41_9658b18e4610e22c9579b9b38269adf2_3 _blueprint_tree_15_72b637784d14b76a881869aee9983fa7_3 _trunk_material_element_9_86d6d8c240e634354ed6179ad68da4f5_2 _leaves_material_element_11_d8f23070444ef180f803ec866de8a069_2 _mesh_material_idformask_23_33f6f1dc43f19a49d55212a45a6b7562 _collisionmesh_27_d515c7ef4923549dd36621918a253b7d _usecustomimpostor_43_b063f0534c360f988d892797fb015e8e _customimpostormesh_46_e40a5445444cf88f0cf789aad2edf530 _flipmaterialidsfornaniteandfallbackmissmatch_48_5b1aa6cd4be3f574e847b290b81e8659) (Utilities|Struct|BreakTreetoConvertData _currenttreedata_3))
  (bind _returnvalue (Utilities|IsValid _staticmesh_tree_fallback_30_b769c69848448d3b2441219dcb6571a6_3 :!class /Script/Engine.KismetSystemLibrary))
  (bind _assetindex_2 (Variables|GetAssetIndex))
  (bind _assetspacing (Variables|Default|GetAssetSpacing))
  (bind _returnvalue_2 (* _assetspacing _assetindex_2))
  (bind _returnvalue_3 (Math|Vector|MakeVector _returnvalue_2 -1.5e+03))
  (bind _returnvalue_4 (Utilities|String|Append _assetbasefolder_7_39c096b443542bc75c0e6387361d3c9a "Impostor/"))
  (bind _returnvalue_5 (Utilities|String|Append "MI_" _treename_9_53616c4448ee21eef59b3a9e0a39aabf))
  (bind _returnvalue_6 (Utilities|String|Append _returnvalue_5 "_Impostor"))
  (bind _returnvalue_7 (Utilities|String|Append _returnvalue_4 _returnvalue_6))
  (bind _geoscript (LocalVariables|GetGeoScript))
  (bind _newbakedstaticmesh (LocalVariables|GetNewBakedStaticmesh))
  (bind _newbakedstaticmesh_2 (LocalVariables|GetNewBakedStaticmesh))
  (bind _geoscript_2 (LocalVariables|GetGeoScript))
  (bind _geoscript_3 (LocalVariables|GetGeoScript))
  (bind _geoscript_4 (LocalVariables|GetGeoScript))
  (bind _assetindex_3 (Variables|GetAssetIndex))
  (LocalVariables|SetCurrentTreeData (Utilities|Array|Get(acopy) (Variables|DontTouchOnlyRead|GetTreeData) AssetIndex))
  (LocalVariables|SetCurrentInternalData (Utilities|Array|Get(acopy) (Variables|Default|GetTree_internalData) AssetIndex))
  (LocalVariables|SetTreeMeshAssetPath (Utilities|String|Append (Utilities|String|Append _assetbasefolder_7_39c096b443542bc75c0e6387361d3c9a_2 "BakeData/") (Utilities|String|Append _treename_9_53616c4448ee21eef59b3a9e0a39aabf_2 "_Bake")))
  (Utilities|FlowControl|Sequence
    (:then_0
      ;[1 1 1] BakeMesh
      (Utilities|FlowControl|Sequence
        (:then_0
          (bind _asbp_geo_script_convert_tree (Utilities|Casting|CastToBP_GeoScript_ConvertTree (Variables|/Script/Engine.ChildActorComponent|GetChildActor (Variables|Default|GetTreeBakeGeoScript))))
          (LocalVariables|SetGeoScript _asbp_geo_script_convert_tree)
          (Variables|/Game/Environments/Asteria/Foliage/Trees/Blueprint_Parent/BP_GeoScript_ConvertTree.BP_GeoScript_ConvertTree_C|SetMode (Variables|ConversionMode|GetMode) _geoscript_4)
          (Variables|/Game/Environments/Asteria/Foliage/Trees/Blueprint_Parent/BP_GeoScript_ConvertTree.BP_GeoScript_ConvertTree_C|SetFlipMateiralIDsForNaniteandFallback _flipmaterialidsfornaniteandfallbackmissmatch_48_5b1aa6cd4be3f574e847b290b81e8659 _geoscript_4)
          (Variables|/Game/Environments/Asteria/Foliage/Trees/Blueprint_Parent/BP_GeoScript_ConvertTree.BP_GeoScript_ConvertTree_C|SetStaticmesh _staticmesh_tree_nanite_31_280c65834a5f8d3da1beea87a2fdfc78_3 _geoscript_3)
          (Variables|/Game/Environments/Asteria/Foliage/Trees/Blueprint_Parent/BP_GeoScript_ConvertTree.BP_GeoScript_ConvertTree_C|SetStaticmesh_Fallback _staticmesh_tree_fallback_30_b769c69848448d3b2441219dcb6571a6_3 _geoscript_3)
          (LocalVariables|SetHasFallbackMesh _returnvalue)
          (LocalVariables|SetHasNaniteMesh (Utilities|IsValid _staticmesh_tree_nanite_31_280c65834a5f8d3da1beea87a2fdfc78_3 :!class /Script/Engine.KismetSystemLibrary))
          (Variables|/Game/Environments/Asteria/Foliage/Trees/Blueprint_Parent/BP_GeoScript_ConvertTree.BP_GeoScript_ConvertTree_C|SetHasFallbackMesh _returnvalue (LocalVariables|GetGeoScript))
          (Variables|/Game/Environments/Asteria/Foliage/Trees/Blueprint_Parent/BP_GeoScript_ConvertTree.BP_GeoScript_ConvertTree_C|SetAsset%20Path%20and%20Name (LocalVariables|GetTreeMeshAssetPath) (LocalVariables|GetGeoScript))
          (Variables|/Game/Environments/Asteria/Foliage/Trees/Blueprint_Parent/BP_GeoScript_ConvertTree.BP_GeoScript_ConvertTree_C|SetTrunkMaterialID _trunk_material_element_9_86d6d8c240e634354ed6179ad68da4f5_2 (LocalVariables|GetGeoScript))
          (Variables|/Game/Environments/Asteria/Foliage/Trees/Blueprint_Parent/BP_GeoScript_ConvertTree.BP_GeoScript_ConvertTree_C|SetLeavesMaterialID _leaves_material_element_11_d8f23070444ef180f803ec866de8a069_2 (LocalVariables|GetGeoScript))
          (Utilities|IsValid _collisionmesh_27_d515c7ef4923549dd36621918a253b7d
            (:Is%20Valid
              (Variables|/Game/Environments/Asteria/Foliage/Trees/Blueprint_Parent/BP_GeoScript_ConvertTree.BP_GeoScript_ConvertTree_C|SetCollisionMesh _collisionmesh_27_d515c7ef4923549dd36621918a253b7d _geoscript)
              (Variables|/Game/Environments/Asteria/Foliage/Trees/Blueprint_Parent/BP_GeoScript_ConvertTree.BP_GeoScript_ConvertTree_C|SetMeshMaterialID _mesh_material_idformask_23_33f6f1dc43f19a49d55212a45a6b7562 _geoscript)
              (Class|BPGeoScriptConvertTree|Update (LocalVariables|GetGeoScript))
              (BakeMesh|BakeStaticmesh (LocalVariables|GetGeoScript)))))
        (:then_1
          ;;[1 1 1] Infuse Custom LODs
          (for (_array_element _array_index) _customlods_38_dc1c72d2449fa58507dc50a878a3d5ae_3
            (Utilities|IsValid _array_element
              (:Is%20Valid
                (Variables|/Game/Environments/Asteria/Foliage/Trees/Blueprint_Parent/BP_GeoScript_ConvertTree.BP_GeoScript_ConvertTree_C|SetTargetLODindex (+ _array_index 1) _geoscript_2)
                (Variables|/Game/Environments/Asteria/Foliage/Trees/Blueprint_Parent/BP_GeoScript_ConvertTree.BP_GeoScript_ConvertTree_C|SetCustomLODMesh _array_element _geoscript_2)
                (BakeMesh|InfuseCustomLOD _geoscript_2)))))))
    (:then_2
      (Utilities|FlowControl|Sequence
        (:then_0
          ;[1 1 1] Spawn Static Mesh Actor for debuging
          (bind _returnvalue_8 (Game|SpawnActorfromClass self "/Script/Engine.StaticMeshActor" (Math|Transform|MakeTransform _returnvalue_3) "Undefined" "MultiplyWithRoot" 0 0))
          ;[1 1 1] Setting for debug mesh
          (Variables|/Script/Engine.Actor|SetTags (Utilities|Array|MakeArray "AutoSpawned") _returnvalue_8)
          (Rendering|Material|SetCustomPrimitiveDataFloat (Variables|/Script/Engine.StaticMeshActor|GetStaticMeshComponent _returnvalue_8) 17 (Math|Conversions|ToFloat(Integer) _assetindex_3))
          (Rendering|Material|SetCustomPrimitiveDataFloat (Variables|/Script/Engine.StaticMeshActor|GetStaticMeshComponent _returnvalue_8) 19 (Math|Conversions|ToFloat(Integer) (Utilities|Array|Get(acopy) (Variables|Default|GetTree_Unique_Index_Leaves) _assetindex_3)))
          (Rendering|Material|SetCustomPrimitiveDataFloat (Variables|/Script/Engine.StaticMeshActor|GetStaticMeshComponent _returnvalue_8) 18 (Math|Conversions|ToFloat(Integer) (Utilities|Array|Get(acopy) (Variables|Default|GetTree_Unique_Index_Trunk) _assetindex_3)))
          (bind _returnvalue_9 (EditorScripting|Asset|LoadAsset (LocalVariables|GetTreeMeshAssetPath) :!class /Script/EditorScriptingUtilities.EditorAssetLibrary))
          (bind _asstatic_mesh (Utilities|Casting|CastToStaticMesh _returnvalue_9))
          (bind _output_get (LocalVariables|SetNewBakedStaticmesh _asstatic_mesh))
          (Components|StaticMesh|SetStaticMesh (Variables|/Script/Engine.StaticMeshActor|GetStaticMeshComponent _returnvalue_8) _output_get)
          (Utilities|FlowControl|Sequence
            (:then_0
              (switch Utilities|FlowControl|Switch|SwitchonConvertTree_Mode (Variables|ConversionMode|GetMode)
                (:NewEnumerator0
                  (Variables|/Script/Engine.StaticMesh|SetStaticMaterials (Variables|/Script/Engine.StaticMesh|GetStaticMaterials (select (LocalVariables|GetHasFallbackMesh) _staticmesh_tree_fallback_30_b769c69848448d3b2441219dcb6571a6_2 _staticmesh_tree_nanite_31_280c65834a5f8d3da1beea87a2fdfc78_2)) (LocalVariables|GetNewBakedStaticmesh))
                  (Utilities|FlowControl|Sequence
                    (:then_0
                      (if (and (LocalVariables|GetHasFallbackMesh) (LocalVariables|GetHasNaniteMesh))
                        (for _array_element_2 (Variables|/Script/Engine.StaticMesh|GetStaticMaterials _staticmesh_tree_fallback_30_b769c69848448d3b2441219dcb6571a6_2)
                          (bind _asmaterial_instance_constant (Utilities|Casting|CastToMaterialInstanceConstant (Utilities|Struct|BreakStaticMaterial _array_element_2)))
                          (if (Math|Boolean|ORBoolean (== (Variables|/Script/Engine.MaterialInstance|GetParent _asmaterial_instance_constant) "/Game/Environments/Asteria/Foliage/Trees/SnowyPine/Materials/M_SnowyPine_master.M_SnowyPine_master") (== (Variables|/Script/Engine.MaterialInstance|GetParent _asmaterial_instance_constant) "/Game/Environments/Asteria/Foliage/Trees/SnowyPine/Materials/M_SnowyPine_Ice_Inst.M_SnowyPine_Ice_Inst") (== (Variables|/Script/Engine.MaterialInstance|GetParent _asmaterial_instance_constant) "/Game/Environments/Asteria/Foliage/Trees/SnowyPine/Materials/M_SnowyPine_Ice_NonNanite.M_SnowyPine_Ice_NonNanite"))
                            (Rendering|Material|SetNaniteOverrideMaterial _asmaterial_instance_constant true "/Game/Environments/Asteria/Foliage/Trees/SnowyPine/Materials/M_SnowyPine_Ice_Inst.M_SnowyPine_Ice_Inst")
                            (else
                              (Rendering|Material|SetNaniteOverrideMaterial _asmaterial_instance_constant true (StaticMesh|GetMaterial _staticmesh_tree_nanite_31_280c65834a5f8d3da1beea87a2fdfc78_2 (Math|Integer|SelectInt _trunk_material_element_9_86d6d8c240e634354ed6179ad68da4f5 _leaves_material_element_11_d8f23070444ef180f803ec866de8a069 (MaterialEditing|GetMaterialInstanceStaticSwitchParameterValue _asmaterial_instance_constant "IsTrunk"))))))
                          (EditorScripting|Asset|SaveLoadedAsset _asmaterial_instance_constant false :!class /Script/EditorScriptingUtilities.EditorAssetLibrary))))
                    (:then_1
                      (StaticMesh|AddMaterial (LocalVariables|GetNewBakedStaticmesh) "/Engine/EngineMaterials/WorldGridMaterial.WorldGridMaterial"))
                    (:then_2
                      ;[1 1 1] SetImpostorMaterial
                      (bind _returnvalue_10 (EditorScripting|Asset|DoesAssetExist _returnvalue_7 :!class /Script/EditorScriptingUtilities.EditorAssetLibrary))
                      (if _returnvalue_10
                        (bind _returnvalue_11 (EditorScripting|Asset|LoadAsset _returnvalue_7 :!class /Script/EditorScriptingUtilities.EditorAssetLibrary))
                        (bind _asmaterial_instance (Utilities|Casting|CastToMaterialInstance _returnvalue_11))
                        (StaticMesh|SetMaterial (LocalVariables|GetNewBakedStaticmesh) (Utilities|Array|LastIndex (Variables|/Script/Engine.StaticMesh|GetStaticMaterials (LocalVariables|GetNewBakedStaticmesh))) _asmaterial_instance)))
                    (:then_3
                      (for _array_element_3 _addcustomlodmaterials_41_9658b18e4610e22c9579b9b38269adf2_2
                        (StaticMesh|AddMaterial (LocalVariables|GetNewBakedStaticmesh) _array_element_3)))))
                (:NewEnumerator1
                  ;[1 1 1] Set Array Texture Material
                  (Variables|/Script/Engine.StaticMesh|SetStaticMaterials (Utilities|Array|MakeArray (Utilities|Struct|MakeStaticMaterial "/Game/Environments/Asteria/Foliage/Trees/MasterMaterials/M_Asteria_Tree_BaseMat_Array.M_Asteria_Tree_BaseMat_Array" "MatSlot1")) _newbakedstaticmesh_2)
                  (StaticMesh|SetMaterial _newbakedstaticmesh_2 0 "/Game/Environments/Asteria/Foliage/Trees/MasterMaterials/M_Asteria_Tree_BaseMat_Array.M_Asteria_Tree_BaseMat_Array"))
                (:NewEnumerator2
                  ;[1 1 1] Set UDIM Material
                  (Variables|/Script/Engine.StaticMesh|SetStaticMaterials (Utilities|Array|MakeArray (Utilities|Struct|MakeStaticMaterial "/Game/Environments/Asteria/Foliage/Trees/MasterMaterials/M_Asteria_Tree_BaseMat_UDIM.M_Asteria_Tree_BaseMat_UDIM" "MatSlot1")) _newbakedstaticmesh)
                  (StaticMesh|SetMaterial _newbakedstaticmesh 0 "/Game/Environments/Asteria/Foliage/Trees/MasterMaterials/M_Asteria_Tree_BaseMat_UDIM.M_Asteria_Tree_BaseMat_UDIM"))))
            (:then_1
              ;[1 1 1] Set Staticmesh on Blueprint and set Custom Prim Data then set Nanite Lod Settings
              (CallFunction|SpawnClass _blueprint_tree_15_72b637784d14b76a881869aee9983fa7 (Math|Transform|MakeTransform (Math|Vector|MakeVector _returnvalue_2 -3e+03)))
              (SetEditorProperties|SetEditorStaticMeshonBlueprint _blueprint_tree_15_72b637784d14b76a881869aee9983fa7 _returnvalue_9)
              (SetEditorProperties|SetEditorCustomPrimDataOnBlueprint _blueprint_tree_15_72b637784d14b76a881869aee9983fa7 _returnvalue_9 (Variables|GetAssetIndex) (Utilities|Array|Get(acopy) (Variables|Default|GetTree_Unique_Index_Leaves) (Variables|GetAssetIndex)) (Utilities|Array|Get(acopy) (Variables|Default|GetTree_Unique_Index_Trunk) (Variables|GetAssetIndex)))
              (SetEditorProperties|SetEditorStaticMeshAttributes _returnvalue_9))))
        (:then_1
          ;[1 1 1] Add Info as text
          (AddComponent|NODE_AddTextRenderComponent-3 :RelativeTransform (Math|Transform|MakeTransform (+ _returnvalue_3 "0.000000,500.000000,0") "90.000000,90.000000, 0"))
          (AddComponent|NODE_AddArrowComponent-1 :RelativeTransform (Math|Transform|MakeTransform _returnvalue_3 "0.000000,-90.000000, 0" "16.000000,16.000000,1.000000")))))))

(fn CreateParameterIndexMapping
  (LocalVariables|SetListText "")
  (Utilities|Array|Clear (Variables|DontTouchOnlyRead|GetParameterIndexList))
  (for (_array_element _array_index) (Variables|Default|GetParameterDataList)
    (Utilities|Array|Add (Variables|DontTouchOnlyRead|GetParameterIndexList) (Utilities|String|StringToName (.ParameterName_4_9C769BB24C9D282C35C078B0E7DB6B92 (Utilities|Struct|BreakTreeParameterData _array_element))))
    (LocalVariables|SetListText (Utilities|String|Append (Utilities|String|Append (Utilities|String|Append (Utilities|String|Append (LocalVariables|GetListText) (Utilities|String|ToString(Integer) _array_index)) "     ") (.ParameterName_4_9C769BB24C9D282C35C078B0E7DB6B92 (Utilities|Struct|BreakTreeParameterData _array_element))) "
")))
  (bind _returnvalue (AddComponent|NODE_AddTextRenderComponent-2))
  (Rendering|Components|TextRender|SetText _returnvalue (Utilities|Text|ToText(String) (LocalVariables|GetListText))))

(fn DebugArrayTextures
  (bind _returnvalue (Math|Vector|MakeVector 0.0 0.0 5e+02))
  (if (Variables|Debug|GetVisualizeArrayTextures)
    (for (_array_element _array_index) (Variables|Default|GetArrayTextureData)
      (if (Variables|Debug|GetVisualizeSingleArray)
        (bind _returnvalue_2 (AddComponent|NODE_AddStaticMeshComponent-3 :RelativeTransform (Math|Transform|MakeTransform (Utilities|Operators|Add :A "0.000000,0,5000.000000" :B (* _returnvalue _array_index)) "0,-9090.000000, 0")))
        (bind _returnvalue_3 (Rendering|Material|CreateDynamicMaterialInstance _returnvalue_2 0 "/Game/Environments/Asteria/Foliage/Trees/MasterMaterials/BakeArrayTexture/M_DebugArrayTexture.M_DebugArrayTexture" "None" :!class /Script/Engine.PrimitiveComponent))
        (Rendering|Material|SetTextureParameterValue _returnvalue_3 "Array" (Utilities|Struct|BreakTreeConvertArrayTextureData _array_element))
        (Rendering|Material|SetScalarParameterValue _returnvalue_3 "Index" (Math|Conversions|ToFloat(Integer) (Variables|Debug|GetVisualizeSingleArray_Index)) :!class /Script/Engine.MaterialInstanceDynamic)
        (else
          (for (_ _array_index_2) (Variables|DontTouchOnlyRead|GetTreeData)
            (bind _returnvalue_4 (AddComponent|NODE_AddStaticMeshComponent-2 :RelativeTransform (Math|Transform|MakeTransform (Utilities|Operators|Add "0.000000,0,5000.000000" (* _returnvalue _array_index) (* (Math|Vector|MakeVector (Variables|Default|GetAssetSpacing)) _array_index_2)) "0,-9090.000000, 0")))
            (bind _returnvalue_5 (Rendering|Material|CreateDynamicMaterialInstance _returnvalue_4 0 "/Game/Environments/Asteria/Foliage/Trees/MasterMaterials/BakeArrayTexture/M_DebugArrayTexture.M_DebugArrayTexture" "None" :!class /Script/Engine.PrimitiveComponent))
            (Rendering|Material|SetTextureParameterValue _returnvalue_5 "Array" (Utilities|Struct|BreakTreeConvertArrayTextureData _array_element))
            (Rendering|Material|SetScalarParameterValue _returnvalue_5 "Index" (Math|Conversions|ToFloat(Integer) _array_index_2) :!class /Script/Engine.MaterialInstanceDynamic)))))))

(fn DeleteallSpawnedActors
  (bind _outactors (Actor|GetAllActorswithTag 0 "AutoSpawned"))
  (for _array_element _outactors
    (Actor|DestroyActor _array_element)))

(fn DubugAttributeTextures
  (bind _assetspacing (Variables|Default|GetAssetSpacing))
  (bind _returnvalue (Math|Vector|MakeVector _assetspacing))
  (if (Variables|Debug|GetDebugAttributeTextures)
    (for (_ _array_index) (Variables|DontTouchOnlyRead|GetTreeData)
      (for (_array_element _array_index_2) (Variables|Default|GetParameterDataList)
        (bind _returnvalue_2 (AddComponent|NODE_AddTextRenderComponent-6 :RelativeTransform (Math|Transform|MakeTransform (+ "-200.000000,0,0.000000" (Utilities|Operators|Add "0.000000,0,5000.000000" (* _returnvalue _array_index) (* "0,0,200.000000" _array_index_2))) "0,90.000000, 0")))
        (Rendering|Components|TextRender|SetText _returnvalue_2 (Utilities|Text|ToText(String) (.ParameterName_4_9C769BB24C9D282C35C078B0E7DB6B92 (Utilities|Struct|BreakTreeParameterData _array_element))))
        (bind _returnvalue_3 (AddComponent|NODE_AddStaticMeshComponent-4 :RelativeTransform (Math|Transform|MakeTransform (Utilities|Operators|Add "0.000000,0,5000.000000" (* _returnvalue _array_index) (* "0,0,200.000000" _array_index_2)) "0,-90.000000, 0")))
        (bind _returnvalue_4 (Rendering|Material|CreateDynamicMaterialInstance _returnvalue_3 0 "/Game/Environments/Asteria/Foliage/Trees/MasterMaterials/BakeArrayTexture/M_DebugAttributeTexture.M_DebugAttributeTexture" "None" :!class /Script/Engine.PrimitiveComponent))
        (Rendering|Material|SetScalarParameterValue _returnvalue_4 "Index" (Math|Conversions|ToFloat(Integer) _array_index_2) :!class /Script/Engine.MaterialInstanceDynamic)
        (Rendering|Material|SetScalarParameterValue _returnvalue_4 "ArrayIndex" (Math|Conversions|ToFloat(Integer) _array_index) :!class /Script/Engine.MaterialInstanceDynamic)))))

(fn ExecuteContructionScript
  (CallFunction|SetTreeDatafromDataAsset)
  (for _array_element (Variables|Default|GetSpawnedComponentList)
    (Utilities|FlowControl|Sequence
      (:then_0
        (bind _asstatic_mesh_component (Utilities|Casting|CastToStaticMeshComponent _array_element))
        (Components|DestroyComponent _asstatic_mesh_component 0))
      (:then_1
        (bind _astext_render_component (Utilities|Casting|CastToTextRenderComponent _array_element))
        (Components|DestroyComponent _astext_render_component 0))))
  (CallFunction|DeleteallSpawnedActors)
  (Utilities|Array|Clear (Variables|Default|GetSpawnedComponentList))
  (Utilities|FlowControl|Sequence
    (:then_0
      (CallFunction|MakeUniqueTreeTextureList))
    (:then_1
      (CallFunction|ProcessDataAssetAndCreateComponents))
    (:then_2
      (CallFunction|CreateParameterIndexMapping))
    (:then_3
      (CallFunction|DebugArrayTextures)
      (CallFunction|DubugAttributeTextures))
    (:then_4
      (CallFunction|ForceShaderCompile))))

(fn ForceShaderCompile
  (bind _returnvalue (AddComponent|NODE_AddStaticMeshComponent-5 :RelativeTransform (Math|Transform|MakeTransform "-5000.000000,-0.000000, 0")))
  (Rendering|Material|CreateDynamicMaterialInstance _returnvalue 0 (Variables|Default|GetAttributeBakeMaterial) "None" :!class /Script/Engine.PrimitiveComponent)
  (bind _returnvalue_2 (AddComponent|NODE_AddStaticMeshComponent-6 :RelativeTransform (Math|Transform|MakeTransform "-5000.000000,-500.000000, 0")))
  (Rendering|Material|CreateDynamicMaterialInstance _returnvalue_2 0 "/Game/Environments/Asteria/Foliage/Trees/MasterMaterials/BakeArrayTexture/M_CopyRT.M_CopyRT" "None" :!class /Script/Engine.PrimitiveComponent)
  (bind _returnvalue_3 (AddComponent|NODE_AddStaticMeshComponent-7 :RelativeTransform (Math|Transform|MakeTransform "-5000.000000,-1000.000000, 0")))
  (Rendering|Material|CreateDynamicMaterialInstance _returnvalue_3 0 "/Game/Environments/Asteria/Foliage/Trees/MasterMaterials/BakeArrayTexture/M_Bake_AnimationTexture.M_Bake_AnimationTexture" "None" :!class /Script/Engine.PrimitiveComponent))

(fn GetAllTextureParametersandMoveToArray
  (params (ArrayData :type Tree%20Convert%20Array%20Texture%20Data%20Structure))
  (bind (_arraytexture_2_f0ae2373441d8d518f66bc98d797a2b1 _parametername_5_2235b6434a89ea3b26f58b96a341eac6 _istrunk_8_eb7fae374b9d640059e4f1a949614119 _isattributetexture_10_45528631445cfe5f0c11c899bc0d69b3 _isanimationtexture_12_fac5ab2b4ccad18caebd9eb829fc30af) (Utilities|Struct|BreakTreeConvertArrayTextureData ArrayData))
  (bind _tree_internaldata (Variables|Default|GetTree_internalData))
  (bind _texturelist (Variables|Default|GetTextureList))
  (bind _tree_internaldata_2 (Variables|Default|GetTree_internalData))
  (Utilities|Array|Clear (Variables|Default|GetTextureList))
  (Utilities|FlowControl|Sequence
    (:then_0
      (if _isanimationtexture_12_fac5ab2b4ccad18caebd9eb829fc30af
        (bind (_maxtexture_res_x _maxtexture_res_y _isvalid) (CallFunction|GetMaxTextureRes _parametername_5_2235b6434a89ea3b26f58b96a341eac6))
        (if _isvalid
          (for (_array_element _array_index) (Variables|DontTouchOnlyRead|GetTreeData)
            (bind _returnvalue (Rendering|Material|CreateDynamicMaterialInstance 0 (StaticMesh|GetMaterial (.StaticMesh_Tree_Nanite_31_280C65834A5F8D3DA1BEEA87A2FDFC78 (Utilities|Struct|BreakTreetoConvertData _array_element)) (.Leaves_Material_Element_11_D8F23070444EF180F803EC866DE8A069 (Utilities|Struct|BreakTreetoConvertData _array_element))) "None" "None" :!class /Script/Engine.KismetMaterialLibrary))
            (bind _returnvalue_2 (Rendering|Material|GetTextureParameterValue _returnvalue _parametername_5_2235b6434a89ea3b26f58b96a341eac6 :!class /Script/Engine.MaterialInstanceDynamic))
            (bind _returnvalue_3 (Rendering|Material|CreateDynamicMaterialInstance 0 "/Game/Environments/Asteria/Foliage/Trees/MasterMaterials/BakeArrayTexture/M_Bake_AnimationTexture.M_Bake_AnimationTexture" "None" "None" :!class /Script/Engine.KismetMaterialLibrary))
            (Rendering|Material|SetTextureParameterValue _returnvalue_3 "Texture" _returnvalue_2)
            (bind _returnvalue_4 (Rendering|CreateRenderTarget2D 0 _maxtexture_res_x _maxtexture_res_y))
            (LocalVariables|SetlocalRenderTarget _returnvalue_4)
            (Rendering|DrawMaterialtoRenderTarget 0 (LocalVariables|GetlocalRenderTarget) _returnvalue_3)
            (Utilities|FlowControl|Sequence
              (:then_0
                (bind _returnvalue_5 (EditorScripting|Asset|DoAssetsExist (Utilities|Array|MakeArray (Utilities|String|Append (Utilities|String|Append (.AssetBaseFolder_7_39C096B443542BC75C0E6387361D3C9A (Utilities|Struct|BreakTreeConverteInternalData (Utilities|Array|Get(acopy) _tree_internaldata_2 _array_index))) "BakeData/") (Utilities|GetDisplayName _returnvalue_2))) :!class /Script/EditorScriptingUtilities.EditorAssetLibrary))
                (if _returnvalue_5
                  (bind _returnvalue_6 (EditorScripting|Asset|LoadAsset (Utilities|String|Append (Utilities|String|Append (.AssetBaseFolder_7_39C096B443542BC75C0E6387361D3C9A (Utilities|Struct|BreakTreeConverteInternalData (Utilities|Array|Get(acopy) _tree_internaldata_2 _array_index))) "BakeData/") (Utilities|GetDisplayName _returnvalue_2)) :!class /Script/EditorScriptingUtilities.EditorAssetLibrary))
                  (bind _astexture_2d (Utilities|Casting|CastToTexture2D _returnvalue_6))
                  (EditorScripting|Rendering|ConvertRenderTargettoTexture2DEditorOnly 0 (LocalVariables|GetlocalRenderTarget) _astexture_2d)
                  (else
                    (EditorScripting|Rendering|RenderTarget2DCreateStaticTexture2DEditorOnly (LocalVariables|GetlocalRenderTarget) (Utilities|String|Append (Utilities|String|Append (.AssetBaseFolder_7_39C096B443542BC75C0E6387361D3C9A (Utilities|Struct|BreakTreeConverteInternalData (Utilities|Array|Get(acopy) _tree_internaldata_2 _array_index))) "BakeData/") (Utilities|GetDisplayName _returnvalue_2)) "TC_HDR" "TMGS_NoMipmaps"))))
              (:then_1
                (Development|PrintString 0 (Utilities|String|Append (Utilities|String|Append (.AssetBaseFolder_7_39C096B443542BC75C0E6387361D3C9A (Utilities|Struct|BreakTreeConverteInternalData (Utilities|Array|Get(acopy) _tree_internaldata_2 _array_index))) "BakeData/") (Utilities|GetDisplayName _returnvalue_2)))
                (Utilities|Array|Add (Variables|Default|GetTextureList) (Utilities|String|Append (Utilities|String|Append (.AssetBaseFolder_7_39C096B443542BC75C0E6387361D3C9A (Utilities|Struct|BreakTreeConverteInternalData (Utilities|Array|Get(acopy) _tree_internaldata_2 _array_index))) "BakeData/") (Utilities|GetDisplayName _returnvalue_2)))))))
        (elif _isattributetexture_10_45528631445cfe5f0c11c899bc0d69b3
          ;[1 1 1] Attribute texture was created and must be refrerenced by folder location string
          (for (_ _array_index_2) (Variables|DontTouchOnlyRead|GetTreeData)
            (bind _returnvalue_7 (EditorScripting|Asset|DoAssetsExist (Utilities|Array|MakeArray (Utilities|String|Append (Utilities|String|Append (.AssetBaseFolder_7_39C096B443542BC75C0E6387361D3C9A (Utilities|Struct|BreakTreeConverteInternalData (Utilities|Array|Get(acopy) _tree_internaldata _array_index_2))) "BakeData/") (Utilities|String|Append (.TreeName_9_53616C4448EE21EEF59B3A9E0A39AABF (Utilities|Struct|BreakTreeConverteInternalData (Utilities|Array|Get(acopy) _tree_internaldata _array_index_2))) "_Attributes"))) :!class /Script/EditorScriptingUtilities.EditorAssetLibrary))
            (if _returnvalue_7
              (bind _returnvalue_8 (EditorScripting|Asset|LoadAsset (Utilities|String|Append (Utilities|String|Append (.AssetBaseFolder_7_39C096B443542BC75C0E6387361D3C9A (Utilities|Struct|BreakTreeConverteInternalData (Utilities|Array|Get(acopy) _tree_internaldata _array_index_2))) "BakeData/") (Utilities|String|Append (.TreeName_9_53616C4448EE21EEF59B3A9E0A39AABF (Utilities|Struct|BreakTreeConverteInternalData (Utilities|Array|Get(acopy) _tree_internaldata _array_index_2))) "_Attributes")) :!class /Script/EditorScriptingUtilities.EditorAssetLibrary))
              (bind _returnvalue_9 (EditorScripting|Asset|GetPathNameforLoadedAsset _returnvalue_8 :!class /Script/EditorScriptingUtilities.EditorAssetLibrary))
              (Utilities|Array|Add (Variables|Default|GetTextureList) _returnvalue_9)
              (else
                (EditorScripting|Asset|LoadAsset "'/Game/Environments/Asteria/Foliage/Trees/MasterMaterials/BakeArrayTexture/Default_Attribute'" :!class /Script/EditorScriptingUtilities.EditorAssetLibrary)
                (bind _returnvalue_10 (EditorScripting|Asset|GetPathNameforLoadedAsset (Variables|Default|GetDefaultAttributeTexture) :!class /Script/EditorScriptingUtilities.EditorAssetLibrary))
                (Utilities|Array|Add (Variables|Default|GetTextureList) _returnvalue_10))))
          (else
            ;[1 1 1] Make List of textures
            (for _array_element_2 (Variables|DontTouchOnlyRead|GetTreeData)
              (bind _returnvalue_11 (Rendering|Material|CreateDynamicMaterialInstance 0 (StaticMesh|GetMaterial (.StaticMesh_Tree_Nanite_31_280C65834A5F8D3DA1BEEA87A2FDFC78 (Utilities|Struct|BreakTreetoConvertData _array_element_2)) (Math|Integer|SelectInt (.Trunk_Material_Element_9_86D6D8C240E634354ED6179AD68DA4F5 (Utilities|Struct|BreakTreetoConvertData _array_element_2)) (.Leaves_Material_Element_11_D8F23070444EF180F803EC866DE8A069 (Utilities|Struct|BreakTreetoConvertData _array_element_2)) _istrunk_8_eb7fae374b9d640059e4f1a949614119)) "None" "None" :!class /Script/Engine.KismetMaterialLibrary))
              (bind _returnvalue_12 (Rendering|Material|GetTextureParameterValue _returnvalue_11 _parametername_5_2235b6434a89ea3b26f58b96a341eac6 :!class /Script/Engine.MaterialInstanceDynamic))
              (bind _returnvalue_13 (EditorScripting|Asset|GetPathNameforLoadedAsset _returnvalue_12 :!class /Script/EditorScriptingUtilities.EditorAssetLibrary))
              (Utilities|Array|Add (Variables|Default|GetTextureList) _returnvalue_13))))))
    (:then_1
      ;[1 1 1] Make Texture Phython string
      (Variables|Default|SetTexturePythonStr "")
      (Variables|Default|SetTexturePythonArrayStr "newarray = [")
      (for (_array_element_3 _array_index_3) _texturelist
        (Utilities|FlowControl|Sequence
          (:then_0
            (Variables|Default|SetTexturePythonStr (Utilities|String|Append (Variables|Default|GetTexturePythonStr) (Utilities|String|Append (Utilities|String|Append (Utilities|String|Append (Utilities|String|Append (Utilities|String|Append (Utilities|String|Append (Utilities|String|Append (Utilities|String|Append (Utilities|String|Append "tex" (Utilities|String|ToString(Integer) _array_index_3)) " = '") _array_element_3) "'
") "texload") (Utilities|String|ToString(Integer) _array_index_3)) " = unreal.load_asset(") (Utilities|String|Append "tex" (Utilities|String|ToString(Integer) _array_index_3))) ",None,True)
"))))
          (:then_1
            (Variables|Default|SetTexturePythonArrayStr (Utilities|String|Append (Variables|Default|GetTexturePythonArrayStr) (Utilities|String|Append (Utilities|String|Append "texload" (Utilities|String|ToString(Integer) _array_index_3)) (select (== _array_index_3 (Utilities|Array|LastIndex _texturelist)) "]
" ","))))))))
    (:then_2
      ;[1 1 1] Array Texture Phyton string
      (bind _returnvalue_14 (EditorScripting|Asset|GetPathNameforLoadedAsset _arraytexture_2_f0ae2373441d8d518f66bc98d797a2b1 :!class /Script/EditorScriptingUtilities.EditorAssetLibrary))
      (Python|Execution|ExecutePythonCommand (Utilities|String|Append "import unreal

texarray =" (Utilities|String|Append (Utilities|String|Append (Utilities|String|Append "'" _returnvalue_14) "'") "
") "texarrayload = unreal.load_asset(texarray,None,True)
" (Variables|Default|GetTexturePythonStr) (Variables|Default|GetTexturePythonArrayStr) "texarrayload.set_editor_property('SourceTextures', newarray)")))))

(fn GetAllTextureParametersandMoveToUDIM
  (params (ArrayData :type Tree%20Convert%20Array%20Texture%20Data%20Structure))
  (bind (_arraytexture_2_f0ae2373441d8d518f66bc98d797a2b1 _parametername_5_2235b6434a89ea3b26f58b96a341eac6 _istrunk_8_eb7fae374b9d640059e4f1a949614119 _isattributetexture_10_45528631445cfe5f0c11c899bc0d69b3 _isanimationtexture_12_fac5ab2b4ccad18caebd9eb829fc30af) (Utilities|Struct|BreakTreeConvertArrayTextureData ArrayData))
  (bind _arraydata (Variables|GetArrayData))
  (bind (_arraytexture_2_f0ae2373441d8d518f66bc98d797a2b1_2 _parametername_5_2235b6434a89ea3b26f58b96a341eac6_2 _istrunk_8_eb7fae374b9d640059e4f1a949614119_2) (Utilities|Struct|BreakTreeConvertArrayTextureData _arraydata))
  (bind _arraydata_2 (Variables|GetArrayData))
  (bind (_arraytexture_2_f0ae2373441d8d518f66bc98d797a2b1_3 _parametername_5_2235b6434a89ea3b26f58b96a341eac6_3) (Utilities|Struct|BreakTreeConvertArrayTextureData _arraydata_2))
  (bind (_arraytexture_2_f0ae2373441d8d518f66bc98d797a2b1_4 _parametername_5_2235b6434a89ea3b26f58b96a341eac6_4 _istrunk_8_eb7fae374b9d640059e4f1a949614119_3 _isattributetexture_10_45528631445cfe5f0c11c899bc0d69b3_2) (Utilities|Struct|BreakTreeConvertArrayTextureData (Variables|GetArrayData)))
  (bind _arraydata_3 (Variables|GetArrayData))
  (bind (_arraytexture_2_f0ae2373441d8d518f66bc98d797a2b1_5 _parametername_5_2235b6434a89ea3b26f58b96a341eac6_5) (Utilities|Struct|BreakTreeConvertArrayTextureData _arraydata_3))
  (bind (_arraytexture_2_f0ae2373441d8d518f66bc98d797a2b1_6 _parametername_5_2235b6434a89ea3b26f58b96a341eac6_6 _istrunk_8_eb7fae374b9d640059e4f1a949614119_4 _isattributetexture_10_45528631445cfe5f0c11c899bc0d69b3_3 _isanimationtexture_12_fac5ab2b4ccad18caebd9eb829fc30af_2) (Utilities|Struct|BreakTreeConvertArrayTextureData (Variables|GetArrayData)))
  (bind _returnvalue (Utilities|String|ToString(Name) _parametername_5_2235b6434a89ea3b26f58b96a341eac6_5))
  (bind _returnvalue_2 (Utilities|String|Append "/Game/Environments/Asteria/Foliage/Trees/MasterMaterials/UDIMTextures/UDIM_Trees_" _returnvalue))
  (bind _tree_internaldata (Variables|Default|GetTree_internalData))
  (bind _texturelist (Variables|Default|GetTextureList))
  (bind _tree_internaldata_2 (Variables|Default|GetTree_internalData))
  (bind _sourcetextures (LocalVariables|GetSourceTextures))
  (Utilities|Array|Clear (LocalVariables|GetSourceTextures))
  (Utilities|Array|Clear (LocalVariables|GetBlockCoords))
  (Utilities|Array|Clear (Variables|Default|GetTextureList))
  (Utilities|FlowControl|Sequence
    (:then_0
      (if _isanimationtexture_12_fac5ab2b4ccad18caebd9eb829fc30af_2
        ;[1 1 1] Animation Texures
        (bind (_maxtexture_res_x _maxtexture_res_y _isvalid) (CallFunction|GetMaxTextureRes _parametername_5_2235b6434a89ea3b26f58b96a341eac6_3 true))
        (if _isvalid
          (for (_array_element _array_index) (Variables|DontTouchOnlyRead|GetTreeData)
            (bind _returnvalue_3 (Rendering|Material|CreateDynamicMaterialInstance 0 (StaticMesh|GetMaterial (.StaticMesh_Tree_Nanite_31_280C65834A5F8D3DA1BEEA87A2FDFC78 (Utilities|Struct|BreakTreetoConvertData _array_element)) (.Leaves_Material_Element_11_D8F23070444EF180F803EC866DE8A069 (Utilities|Struct|BreakTreetoConvertData _array_element))) "None" "None" :!class /Script/Engine.KismetMaterialLibrary))
            (bind _returnvalue_4 (Rendering|Material|GetTextureParameterValue _returnvalue_3 _parametername_5_2235b6434a89ea3b26f58b96a341eac6_3 :!class /Script/Engine.MaterialInstanceDynamic))
            (bind _returnvalue_5 (Rendering|Material|CreateDynamicMaterialInstance 0 "/Game/Environments/Asteria/Foliage/Trees/MasterMaterials/BakeArrayTexture/M_Bake_AnimationTexture.M_Bake_AnimationTexture" "None" "None" :!class /Script/Engine.KismetMaterialLibrary))
            (Rendering|Material|SetTextureParameterValue _returnvalue_5 "Texture" _returnvalue_4)
            (bind _returnvalue_6 (Rendering|CreateRenderTarget2D 0 _maxtexture_res_x _maxtexture_res_y))
            (LocalVariables|SetlocalRenderTarget _returnvalue_6)
            (Rendering|DrawMaterialtoRenderTarget 0 (LocalVariables|GetlocalRenderTarget) _returnvalue_5)
            (Utilities|FlowControl|Sequence
              (:then_0
                (bind _returnvalue_7 (EditorScripting|Asset|DoAssetsExist (Utilities|Array|MakeArray (Utilities|String|Append (Utilities|String|Append (.AssetBaseFolder_7_39C096B443542BC75C0E6387361D3C9A (Utilities|Struct|BreakTreeConverteInternalData (Utilities|Array|Get(acopy) _tree_internaldata_2 _array_index))) "BakeData/") (Utilities|GetDisplayName _returnvalue_4))) :!class /Script/EditorScriptingUtilities.EditorAssetLibrary))
                (if _returnvalue_7
                  (bind _returnvalue_8 (EditorScripting|Asset|LoadAsset (Utilities|String|Append (Utilities|String|Append (.AssetBaseFolder_7_39C096B443542BC75C0E6387361D3C9A (Utilities|Struct|BreakTreeConverteInternalData (Utilities|Array|Get(acopy) _tree_internaldata_2 _array_index))) "BakeData/") (Utilities|GetDisplayName _returnvalue_4)) :!class /Script/EditorScriptingUtilities.EditorAssetLibrary))
                  (bind _astexture_2d (Utilities|Casting|CastToTexture2D _returnvalue_8))
                  (EditorScripting|Rendering|ConvertRenderTargettoTexture2DEditorOnly 0 (LocalVariables|GetlocalRenderTarget) _astexture_2d)
                  (Variables|/Script/Engine.Texture|SetCompressionSettings "TC_HDR" _astexture_2d)
                  (else
                    (EditorScripting|Rendering|RenderTarget2DCreateStaticTexture2DEditorOnly (LocalVariables|GetlocalRenderTarget) (Utilities|String|Append (Utilities|String|Append (.AssetBaseFolder_7_39C096B443542BC75C0E6387361D3C9A (Utilities|Struct|BreakTreeConverteInternalData (Utilities|Array|Get(acopy) _tree_internaldata_2 _array_index))) "BakeData/") (Utilities|GetDisplayName _returnvalue_4)) "TC_HDR" "TMGS_NoMipmaps"))))
              (:then_1
                (Development|PrintString 0 (Utilities|String|Append (Utilities|String|Append (.AssetBaseFolder_7_39C096B443542BC75C0E6387361D3C9A (Utilities|Struct|BreakTreeConverteInternalData (Utilities|Array|Get(acopy) _tree_internaldata_2 _array_index))) "BakeData/") (Utilities|GetDisplayName _returnvalue_4)))
                (Utilities|Array|Add (Variables|Default|GetTextureList) (Utilities|String|Append (Utilities|String|Append (.AssetBaseFolder_7_39C096B443542BC75C0E6387361D3C9A (Utilities|Struct|BreakTreeConverteInternalData (Utilities|Array|Get(acopy) _tree_internaldata_2 _array_index))) "BakeData/") (Utilities|GetDisplayName _returnvalue_4)))
                (bind _returnvalue_9 (EditorScripting|Asset|LoadAsset (Utilities|String|Append (Utilities|String|Append (.AssetBaseFolder_7_39C096B443542BC75C0E6387361D3C9A (Utilities|Struct|BreakTreeConverteInternalData (Utilities|Array|Get(acopy) _tree_internaldata_2 _array_index))) "BakeData/") (Utilities|GetDisplayName _returnvalue_4)) :!class /Script/EditorScriptingUtilities.EditorAssetLibrary))
                (bind _astexture_2d_2 (Utilities|Casting|CastToTexture2D _returnvalue_9))
                (Utilities|Array|Add (LocalVariables|GetSourceTextures) _astexture_2d_2)
                (Utilities|Array|Add (LocalVariables|GetBlockCoords) (Utilities|Struct|MakeIntPoint _array_index))))))
        (elif _isattributetexture_10_45528631445cfe5f0c11c899bc0d69b3_2
          ;[1 1 1] Attribute texture was created and must be refrerenced by folder location string
          (for (_ _array_index_2) (Variables|DontTouchOnlyRead|GetTreeData)
            (bind _returnvalue_10 (EditorScripting|Asset|DoAssetsExist (Utilities|Array|MakeArray (Utilities|String|Append (Utilities|String|Append (.AssetBaseFolder_7_39C096B443542BC75C0E6387361D3C9A (Utilities|Struct|BreakTreeConverteInternalData (Utilities|Array|Get(acopy) _tree_internaldata _array_index_2))) "BakeData/") (Utilities|String|Append (.TreeName_9_53616C4448EE21EEF59B3A9E0A39AABF (Utilities|Struct|BreakTreeConverteInternalData (Utilities|Array|Get(acopy) _tree_internaldata _array_index_2))) "_Attributes"))) :!class /Script/EditorScriptingUtilities.EditorAssetLibrary))
            (Utilities|FlowControl|Sequence
              (:then_0
                ;[1 1 1] UDIM
                (if _returnvalue_10
                  (bind _returnvalue_11 (EditorScripting|Asset|LoadAsset (Utilities|String|Append (Utilities|String|Append (.AssetBaseFolder_7_39C096B443542BC75C0E6387361D3C9A (Utilities|Struct|BreakTreeConverteInternalData (Utilities|Array|Get(acopy) _tree_internaldata _array_index_2))) "BakeData/") (Utilities|String|Append (.TreeName_9_53616C4448EE21EEF59B3A9E0A39AABF (Utilities|Struct|BreakTreeConverteInternalData (Utilities|Array|Get(acopy) _tree_internaldata _array_index_2))) "_Attributes")) :!class /Script/EditorScriptingUtilities.EditorAssetLibrary))
                  (bind _astexture_2d_3 (Utilities|Casting|CastToTexture2D _returnvalue_11))
                  (Utilities|Array|Add (LocalVariables|GetSourceTextures) _astexture_2d_3)
                  (else
                    (Utilities|Array|Add (LocalVariables|GetSourceTextures) (Variables|Default|GetDefaultAttributeTexture))))
                (Utilities|Array|Add (LocalVariables|GetBlockCoords) (Utilities|Struct|MakeIntPoint _array_index_2)))
              (:then_1
                ;[1 1 1] ArrayTexture because UDIM become huge
                (if _returnvalue_10
                  (bind _returnvalue_12 (EditorScripting|Asset|LoadAsset (Utilities|String|Append (Utilities|String|Append (.AssetBaseFolder_7_39C096B443542BC75C0E6387361D3C9A (Utilities|Struct|BreakTreeConverteInternalData (Utilities|Array|Get(acopy) _tree_internaldata _array_index_2))) "BakeData/") (Utilities|String|Append (.TreeName_9_53616C4448EE21EEF59B3A9E0A39AABF (Utilities|Struct|BreakTreeConverteInternalData (Utilities|Array|Get(acopy) _tree_internaldata _array_index_2))) "_Attributes")) :!class /Script/EditorScriptingUtilities.EditorAssetLibrary))
                  (bind _returnvalue_13 (EditorScripting|Asset|GetPathNameforLoadedAsset _returnvalue_12 :!class /Script/EditorScriptingUtilities.EditorAssetLibrary))
                  (Utilities|Array|Add (Variables|Default|GetTextureList) _returnvalue_13)
                  (else
                    (EditorScripting|Asset|LoadAsset "'/Game/Environments/Asteria/Foliage/Trees/MasterMaterials/BakeArrayTexture/Default_Attribute'" :!class /Script/EditorScriptingUtilities.EditorAssetLibrary)
                    (bind _returnvalue_14 (EditorScripting|Asset|GetPathNameforLoadedAsset (Variables|Default|GetDefaultAttributeTexture) :!class /Script/EditorScriptingUtilities.EditorAssetLibrary))
                    (Utilities|Array|Add (Variables|Default|GetTextureList) _returnvalue_14))))))
          (else
            ;[1 0.9481 0.9068] Make List of textures Optimized
            (LocalVariables|SetCurrentUDIMIndex 0)
            (for (_array_element_2 _array_index_3) (Variables|DontTouchOnlyRead|GetTreeData)
              (if (< (Utilities|Array|LastIndex (LocalVariables|GetSourceTextures)) (select _istrunk_8_eb7fae374b9d640059e4f1a949614119_2 (Utilities|Array|Get(acopy) (Variables|Default|GetTree_Unique_Index_Trunk) _array_index_3) (Utilities|Array|Get(acopy) (Variables|Default|GetTree_Unique_Index_Leaves) _array_index_3)))
                (bind _returnvalue_15 (Rendering|Material|CreateDynamicMaterialInstance 0 (StaticMesh|GetMaterial (.StaticMesh_Tree_Nanite_31_280C65834A5F8D3DA1BEEA87A2FDFC78 (Utilities|Struct|BreakTreetoConvertData _array_element_2)) (Math|Integer|SelectInt (.Trunk_Material_Element_9_86D6D8C240E634354ED6179AD68DA4F5 (Utilities|Struct|BreakTreetoConvertData _array_element_2)) (.Leaves_Material_Element_11_D8F23070444EF180F803EC866DE8A069 (Utilities|Struct|BreakTreetoConvertData _array_element_2)) _istrunk_8_eb7fae374b9d640059e4f1a949614119_2)) "None" "None" :!class /Script/Engine.KismetMaterialLibrary))
                (bind _returnvalue_16 (Rendering|Material|GetTextureParameterValue _returnvalue_15 _parametername_5_2235b6434a89ea3b26f58b96a341eac6_2 :!class /Script/Engine.MaterialInstanceDynamic))
                (Utilities|FlowControl|Sequence
                  (:then_0
                    (bind _returnvalue_17 (EditorScripting|Asset|GetPathNameforLoadedAsset _returnvalue_16 :!class /Script/EditorScriptingUtilities.EditorAssetLibrary))
                    (Utilities|Array|Add (Variables|Default|GetTextureList) _returnvalue_17))
                  (:then_1
                    (bind _astexture_2d_4 (Utilities|Casting|CastToTexture2D _returnvalue_16))
                    (Utilities|Array|Add _sourcetextures _astexture_2d_4)
                    (Utilities|Array|Add (LocalVariables|GetBlockCoords) (Utilities|Struct|MakeIntPoint (Utilities|Array|LastIndex _sourcetextures)))
                    (Math|Integer|IncrementInt (LocalVariables|GetCurrentUDIMIndex))))))))))
    (:then_1
      ;[1 1 1] Make Texture Phython string
      (Variables|Default|SetTexturePythonStr "")
      (Variables|Default|SetTexturePythonArrayStr "newarray = [")
      (for (_array_element_3 _array_index_4) _texturelist
        (Utilities|FlowControl|Sequence
          (:then_0
            (Variables|Default|SetTexturePythonStr (Utilities|String|Append (Variables|Default|GetTexturePythonStr) (Utilities|String|Append (Utilities|String|Append (Utilities|String|Append (Utilities|String|Append (Utilities|String|Append (Utilities|String|Append (Utilities|String|Append (Utilities|String|Append (Utilities|String|Append "tex" (Utilities|String|ToString(Integer) _array_index_4)) " = '") _array_element_3) "'
") "texload") (Utilities|String|ToString(Integer) _array_index_4)) " = unreal.load_asset(") (Utilities|String|Append "tex" (Utilities|String|ToString(Integer) _array_index_4))) ",None,True)
"))))
          (:then_1
            (Variables|Default|SetTexturePythonArrayStr (Utilities|String|Append (Variables|Default|GetTexturePythonArrayStr) (Utilities|String|Append (Utilities|String|Append "texload" (Utilities|String|ToString(Integer) _array_index_4)) (select (== _array_index_4 (Utilities|Array|LastIndex _texturelist)) "]
" ","))))))))
    (:then_2
      (if (Utilities|Enum|Equal(Enum) (Variables|ConversionMode|GetMode) "NewEnumerator1")
        _
        (elif (or _isattributetexture_10_45528631445cfe5f0c11c899bc0d69b3 _isanimationtexture_12_fac5ab2b4ccad18caebd9eb829fc30af)
          _
          (else
            ;[1 1 1] Create UDIM
            (bind _asutility_make_udim (Utilities|Casting|CastToUtility_MakeUDIM (Variables|/Script/Engine.ChildActorComponent|GetChildActor (Variables|Default|GetChildActor_MakeUDIM))))
            (Class|UtilityMakeUDIM|CreateUDIM _asutility_make_udim _returnvalue_2 (LocalVariables|GetSourceTextures) (LocalVariables|GetBlockCoords) false false)
            (Development|PrintString 0 (Utilities|String|Append "CreatingUDIM = " _returnvalue_2)))))
      ;[1 1 1] Array Texture Phyton string
      (bind _returnvalue_18 (EditorScripting|Asset|GetPathNameforLoadedAsset _arraytexture_2_f0ae2373441d8d518f66bc98d797a2b1 :!class /Script/EditorScriptingUtilities.EditorAssetLibrary))
      (Python|Execution|ExecutePythonCommand (Utilities|String|Append "import unreal

texarray =" (Utilities|String|Append (Utilities|String|Append (Utilities|String|Append "'" _returnvalue_18) "'") "
") "texarrayload = unreal.load_asset(texarray,None,True)
" (Variables|Default|GetTexturePythonStr) (Variables|Default|GetTexturePythonArrayStr) "texarrayload.set_editor_property('SourceTextures', newarray)")))))

(fn GetMaxTextureRes
  (params (ParameterName :type Name) (Local%20Debug :type Boolean))
  (Utilities|FlowControl|Sequence
    (:then_0
      (LocalVariables|SetMaxRes_X 0)
      (LocalVariables|SetMaxRes_Y 0)
      (for _array_element (Variables|DontTouchOnlyRead|GetTreeData)
        (bind _returnvalue (Rendering|Material|CreateDynamicMaterialInstance 0 (StaticMesh|GetMaterial (.StaticMesh_Tree_Nanite_31_280C65834A5F8D3DA1BEEA87A2FDFC78 (Utilities|Struct|BreakTreetoConvertData _array_element)) (.Leaves_Material_Element_11_D8F23070444EF180F803EC866DE8A069 (Utilities|Struct|BreakTreetoConvertData _array_element))) "None" "None" :!class /Script/Engine.KismetMaterialLibrary))
        (bind _returnvalue_2 (Rendering|Material|GetTextureParameterValue _returnvalue ParameterName :!class /Script/Engine.MaterialInstanceDynamic))
        (bind _astexture_2d (Utilities|Casting|CastToTexture2D _returnvalue_2)
          (:then
            (if (Variables|GetLocal%20Debug)
              (Development|PrintString 0 (Utilities|String|Append (Utilities|String|Append "BAX_ TEX = " (Utilities|GetDisplayName _returnvalue_2)) (Utilities|String|Append (Utilities|String|Append "     _ ReS X = " (Utilities|String|ToString(Integer) (Rendering|Texture|GetSizeX _astexture_2d))) (Utilities|String|Append "  Y = " (Utilities|String|ToString(Integer) (Rendering|Texture|GetSizeY _astexture_2d)))))))
            (LocalVariables|SetMaxRes_X (Math|Integer|Max(Integer) (Rendering|Texture|GetSizeX _astexture_2d) (LocalVariables|GetMaxRes_X)))
            (LocalVariables|SetMaxRes_Y (Math|Integer|Max(Integer) (LocalVariables|GetMaxRes_Y) (Rendering|Texture|GetSizeY _astexture_2d)))
            (LocalVariables|SetIsSuccessfull true))
          (:CastFailed
            (LocalVariables|SetIsSuccessfull false)))))
    (:then_1
      (if (Variables|GetLocal%20Debug)
        (Development|PrintString 0 (Utilities|String|Append (Utilities|String|Append (Utilities|String|Append "BAX_FINAL TEX = " (Utilities|String|ToString(Name) ParameterName)) (Utilities|String|Append " Res X  is " (Utilities|String|ToString(Integer) (Math|Float|Truncate (Math|Float|Power 2.0 (Math|Conversions|ToFloat(Integer) (Math|Float|Ceil (/ (Math|Float|Loge (Math|Conversions|ToFloat(Integer) (LocalVariables|GetMaxRes_X))) (Math|Float|Loge 2.0))))))))) (Utilities|String|Append " Res Y  is " (Utilities|String|ToString(Integer) (Math|Float|Truncate (Math|Float|Power 2.0 (Math|Conversions|ToFloat(Integer) (Math|Float|Ceil (/ (Math|Float|Loge (Math|Conversions|ToFloat(Integer) (LocalVariables|GetMaxRes_Y))) (Math|Float|Loge 2.0)))))))))))
      (return (LocalVariables|GetMaxRes_X) (LocalVariables|GetMaxRes_Y) (LocalVariables|GetIsSuccessfull)))))

(fn MakeUniqueTreeTextureList
  (bind _uniquetexturelist_leaves (Variables|Default|GetUniqueTextureList_Leaves))
  (bind _uniquetexturelist_trunk (Variables|Default|GetUniqueTextureList_Trunk))
  (Utilities|Array|Clear (Variables|Default|GetUniqueTextureList_Leaves))
  (Utilities|Array|Clear (Variables|Default|GetUniqueTextureList_Trunk))
  (Utilities|Array|Clear (Variables|Default|GetTree_Unique_Index_Leaves))
  (Utilities|Array|Clear (Variables|Default|GetTree_Unique_Index_Trunk))
  (for _array_element (Variables|DontTouchOnlyRead|GetTreeData)
    (Utilities|FlowControl|Sequence
      (:then_0
        (bind _returnvalue (Rendering|Material|CreateDynamicMaterialInstance 0 (StaticMesh|GetMaterial (.StaticMesh_Tree_Nanite_31_280C65834A5F8D3DA1BEEA87A2FDFC78 (Utilities|Struct|BreakTreetoConvertData _array_element)) (.Leaves_Material_Element_11_D8F23070444EF180F803EC866DE8A069 (Utilities|Struct|BreakTreetoConvertData _array_element))) "None" "None" :!class /Script/Engine.KismetMaterialLibrary))
        (bind _returnvalue_2 (Rendering|Material|GetTextureParameterValue _returnvalue "Diffuse" :!class /Script/Engine.MaterialInstanceDynamic))
        (Utilities|Array|AddUnique _uniquetexturelist_leaves _returnvalue_2)
        (Utilities|Array|Add (Variables|Default|GetTree_Unique_Index_Leaves) (Utilities|Array|FindItem _uniquetexturelist_leaves _returnvalue_2)))
      (:then_1
        (bind _returnvalue_3 (Rendering|Material|CreateDynamicMaterialInstance 0 (StaticMesh|GetMaterial (.StaticMesh_Tree_Nanite_31_280C65834A5F8D3DA1BEEA87A2FDFC78 (Utilities|Struct|BreakTreetoConvertData _array_element)) (.Trunk_Material_Element_9_86D6D8C240E634354ED6179AD68DA4F5 (Utilities|Struct|BreakTreetoConvertData _array_element))) "None" "None" :!class /Script/Engine.KismetMaterialLibrary))
        (bind _returnvalue_4 (Rendering|Material|GetTextureParameterValue _returnvalue_3 "BaseColor_Trunk" :!class /Script/Engine.MaterialInstanceDynamic))
        (Utilities|Array|AddUnique _uniquetexturelist_trunk _returnvalue_4)
        (Utilities|Array|Add (Variables|Default|GetTree_Unique_Index_Trunk) (Utilities|Array|FindItem _uniquetexturelist_trunk _returnvalue_4))))))

(fn ProcessDataAssetandCreateComponents
  (bind _currentdata (LocalVariables|GetCurrentData))
  (bind (_outputname_34_04cb671a4066ad0bd2f6678c3be2ba7a _staticmesh_tree_nanite_31_280c65834a5f8d3da1beea87a2fdfc78) (Utilities|Struct|BreakTreetoConvertData _currentdata))
  (bind (_outputname_34_04cb671a4066ad0bd2f6678c3be2ba7a_2 _staticmesh_tree_nanite_31_280c65834a5f8d3da1beea87a2fdfc78_2) (Utilities|Struct|BreakTreetoConvertData (LocalVariables|GetCurrentData)))
  (bind _currentdata_2 (LocalVariables|GetCurrentData))
  (bind (_outputname_34_04cb671a4066ad0bd2f6678c3be2ba7a_3 _staticmesh_tree_nanite_31_280c65834a5f8d3da1beea87a2fdfc78_3 _staticmesh_tree_fallback_30_b769c69848448d3b2441219dcb6571a6 _customlods_38_dc1c72d2449fa58507dc50a878a3d5ae _addcustomlodmaterials_41_9658b18e4610e22c9579b9b38269adf2 _blueprint_tree_15_72b637784d14b76a881869aee9983fa7 _trunk_material_element_9_86d6d8c240e634354ed6179ad68da4f5 _leaves_material_element_11_d8f23070444ef180f803ec866de8a069) (Utilities|Struct|BreakTreetoConvertData _currentdata_2))
  (bind _returnvalue (StaticMesh|GetMaterial _staticmesh_tree_nanite_31_280c65834a5f8d3da1beea87a2fdfc78_3 _leaves_material_element_11_d8f23070444ef180f803ec866de8a069))
  (bind _returnvalue_2 (StaticMesh|GetMaterial _staticmesh_tree_nanite_31_280c65834a5f8d3da1beea87a2fdfc78_3 _trunk_material_element_9_86d6d8c240e634354ed6179ad68da4f5))
  (bind _assetspacing (Variables|Default|GetAssetSpacing))
  (bind _spawnedcomponentlist (Variables|Default|GetSpawnedComponentList))
  (Utilities|Array|Clear (Variables|Default|GetTree_internalData))
  (for (_array_element _array_index) (Variables|DontTouchOnlyRead|GetTreeData)
    (LocalVariables|SetCurrentData _array_element)
    (Utilities|FlowControl|Sequence
      (:then_0
        (bind _returnvalue_3 (AddComponent|NODE_AddStaticMeshComponent-0 :RelativeTransform (Math|Transform|MakeTransform (Math|Vector|MakeVector (* _assetspacing _array_index)) "0, 0, 0" "10.000000,10.000000,10.000000")))
        (bind _returnvalue_4 (AddComponent|NODE_AddTextRenderComponent-0 :RelativeTransform (Math|Transform|MakeTransform (+ (Math|Vector|MakeVector (* _assetspacing _array_index)) "0.000000,500.000000,50.000000") "90.000000,90.000000, 0")))
        (bind _returnvalue_5 (AddComponent|NODE_AddTextRenderComponent-1 :RelativeTransform (Math|Transform|MakeTransform (+ (Math|Vector|MakeVector (* _assetspacing _array_index)) "0.000000,-2500.000000,50.000000") "90.000000,90.000000, 0")))
        (bind _returnvalue_6 (AddComponent|NODE_AddTextRenderComponent-4 :RelativeTransform (Math|Transform|MakeTransform (+ (Math|Vector|MakeVector (* _assetspacing _array_index)) "0.000000,500.000000,500.000000") "0,90.000000, 0")))
        (Rendering|Components|TextRender|SetText _returnvalue_6 (Utilities|Text|ToText(String) (Utilities|String|Append "ARRAY SOURCE TEXTURE INDEX




")))
        (bind _returnvalue_7 (AddComponent|NODE_AddTextRenderComponent-5 :RelativeTransform (Math|Transform|MakeTransform (+ (Math|Vector|MakeVector (* _assetspacing _array_index)) "0.000000,500.000000,500.000000") "0,90.000000, 0")))
        (Rendering|Components|TextRender|SetText _returnvalue_7 (Utilities|Text|ToText(String) (Utilities|String|ToString(Integer) _array_index)))
        (bind _returnvalue_8 (AddComponent|NODE_AddStaticMeshComponent-1 :RelativeTransform (Math|Transform|MakeTransform (Math|Vector|MakeVector (* _assetspacing _array_index)))))
        (Components|StaticMesh|SetStaticMesh _returnvalue_8 _staticmesh_tree_nanite_31_280c65834a5f8d3da1beea87a2fdfc78_2)
        (Utilities|Array|Add _spawnedcomponentlist _returnvalue_3)
        (Utilities|Array|Add _spawnedcomponentlist _returnvalue_4)
        (Utilities|Array|Add _spawnedcomponentlist _returnvalue_5)
        (Utilities|Array|Add _spawnedcomponentlist _returnvalue_5)
        (Utilities|Array|Add _spawnedcomponentlist _returnvalue_6)
        (Utilities|Array|Add _spawnedcomponentlist _returnvalue_8)
        (Utilities|Array|Add _spawnedcomponentlist _returnvalue_7))
      (:then_1
        (Utilities|FlowControl|Sequence
          (:then_0
            (Utilities|IsValid _returnvalue_2
              (:Is%20Valid
                (bind _returnvalue_9 (Rendering|Material|CreateDynamicMaterialInstance 0 _returnvalue_2 "None" "None" :!class /Script/Engine.KismetMaterialLibrary)))))
          (:then_1
            (Utilities|IsValid _returnvalue
              (:Is%20Valid
                (bind _returnvalue_10 (Rendering|Material|CreateDynamicMaterialInstance 0 _returnvalue "None" "None" :!class /Script/Engine.KismetMaterialLibrary)))))
          (:then_2
            (bind _returnvalue_11 (EditorScripting|Asset|GetPathNameforLoadedAsset _staticmesh_tree_nanite_31_280c65834a5f8d3da1beea87a2fdfc78 :!class /Script/EditorScriptingUtilities.EditorAssetLibrary))
            (Development|PrintString 0 (Utilities|String|GetSubstring (Utilities|String|GetSubstring _returnvalue_11 0 (Utilities|String|FindSubstring _returnvalue_11 "/" false true)) 0 (+ (Utilities|String|FindSubstring (Utilities|String|GetSubstring _returnvalue_11 0 (Utilities|String|FindSubstring _returnvalue_11 "/" false true)) "/" false true) 1)))
            (Development|PrintString 0 (Utilities|String|GetSubstring (Utilities|String|GetSubstring _returnvalue_11 (+ (Utilities|String|FindSubstring _returnvalue_11 "/" false true) 1) (Utilities|String|Len _returnvalue_11)) 0 (Utilities|String|FindSubstring (Utilities|String|GetSubstring _returnvalue_11 (+ (Utilities|String|FindSubstring _returnvalue_11 "/" false true) 1) (Utilities|String|Len _returnvalue_11)) "." false true))))
          (:then_3
            (Utilities|Array|Add (Variables|Default|GetTree_internalData) (Utilities|Struct|MakeTreeConverteInternalData _returnvalue_9 _returnvalue_10 (Utilities|String|GetSubstring (Utilities|String|GetSubstring _returnvalue_11 0 (Utilities|String|FindSubstring _returnvalue_11 "/" false true)) 0 (+ (Utilities|String|FindSubstring (Utilities|String|GetSubstring _returnvalue_11 0 (Utilities|String|FindSubstring _returnvalue_11 "/" false true)) "/" false true) 1)) (Utilities|String|SelectString (Utilities|String|GetSubstring (Utilities|String|GetSubstring _returnvalue_11 (+ (Utilities|String|FindSubstring _returnvalue_11 "/" false true) 1) (Utilities|String|Len _returnvalue_11)) 0 (Utilities|String|FindSubstring (Utilities|String|GetSubstring _returnvalue_11 (+ (Utilities|String|FindSubstring _returnvalue_11 "/" false true) 1) (Utilities|String|Len _returnvalue_11)) "." false true)) _outputname_34_04cb671a4066ad0bd2f6678c3be2ba7a (Utilities|String|EqualExactly(String) _outputname_34_04cb671a4066ad0bd2f6678c3be2ba7a))))))))))

(fn RefreshTreeData
  (Variables|Default|SetTreeDataAsset (Variables|Default|GetTreeDataAsset))
  (CallFunction|ExecuteContructionScript)
  (Transformation|SetActorLocation "1.000000,0.000000,0.000000")
  (Transformation|SetActorLocation "0, 0, 0"))

(fn SetTexturesFromMode
  (for _array_element (Variables|Default|GetArrayTextureData)
    (Default|GetAllTextureParametersandMovetoUDIM _array_element)))

(fn SetArrayTextures_Deprecated
  (for _array_element (Variables|Default|GetArrayTextureData)
    (CallFunction|GetAllTextureParametersandMovetoArray _array_element)))

(fn SetEditorCustomPrimDataonBlueprint
  (params (Tree%20Blueprint%20Class :type Object%20Class%20Reference) (LoadedAsset%20Staticmesh :type Object%20Reference) (ArrayIndex :type Integer) (ArrayIndex_Leaves :type Integer) (ArrayIndex_Trunk :type Integer))
  (bind _returnvalue (Utilities|GetClassDisplayName Tree%20Blueprint%20Class))
  (bind _returnvalue_2 (Utilities|String|FindSubstring _returnvalue "_" false true))
  (bind _returnvalue_3 (Utilities|String|GetSubstring _returnvalue 0 _returnvalue_2))
  (bind _returnvalue_4 (Utilities|String|Append _returnvalue_3 "."))
  (bind _returnvalue_5 (Utilities|String|Append _returnvalue_4 _returnvalue))
  (bind _returnvalue_6 (EditorScripting|Asset|GetPathNameforLoadedAsset LoadedAsset%20Staticmesh :!class /Script/EditorScriptingUtilities.EditorAssetLibrary))
  (Python|Execution|ExecutePythonCommand (Utilities|String|Append "import unreal

asset = '" (Utilities|String|Append (Utilities|String|Append (Utilities|String|GetSubstring (Utilities|String|GetSubstring _returnvalue_6 0 (Utilities|String|FindSubstring _returnvalue_6 "/" false true)) 0 (+ (Utilities|String|FindSubstring (Utilities|String|GetSubstring _returnvalue_6 0 (Utilities|String|FindSubstring _returnvalue_6 "/" false true)) "/" false true) 1)) "Blueprint/") _returnvalue_5) "'

index = " (Utilities|String|ToString(Integer) ArrayIndex) "

bp = unreal.load_class(None, asset)

blueprint_class_default = unreal.get_default_object(bp)

#smload = unreal.load_asset(sm,None,True)

blueprint_class_default.set_editor_property('ArrayIndex', index)
"))
  (Python|Execution|ExecutePythonCommand (Utilities|String|Append "import unreal

asset = '" (Utilities|String|Append (Utilities|String|Append (Utilities|String|GetSubstring (Utilities|String|GetSubstring _returnvalue_6 0 (Utilities|String|FindSubstring _returnvalue_6 "/" false true)) 0 (+ (Utilities|String|FindSubstring (Utilities|String|GetSubstring _returnvalue_6 0 (Utilities|String|FindSubstring _returnvalue_6 "/" false true)) "/" false true) 1)) "Blueprint/") _returnvalue_5) "'

index = " (Utilities|String|ToString(Integer) ArrayIndex_Trunk) "

bp = unreal.load_class(None, asset)

blueprint_class_default = unreal.get_default_object(bp)

#smload = unreal.load_asset(sm,None,True)

blueprint_class_default.set_editor_property('ArrayIndex_Trunk', index)
"))
  (Python|Execution|ExecutePythonCommand (Utilities|String|Append "import unreal

asset = '" (Utilities|String|Append (Utilities|String|Append (Utilities|String|GetSubstring (Utilities|String|GetSubstring _returnvalue_6 0 (Utilities|String|FindSubstring _returnvalue_6 "/" false true)) 0 (+ (Utilities|String|FindSubstring (Utilities|String|GetSubstring _returnvalue_6 0 (Utilities|String|FindSubstring _returnvalue_6 "/" false true)) "/" false true) 1)) "Blueprint/") _returnvalue_5) "'

index = " (Utilities|String|ToString(Integer) ArrayIndex_Leaves) "

bp = unreal.load_class(None, asset)

blueprint_class_default = unreal.get_default_object(bp)

#smload = unreal.load_asset(sm,None,True)

blueprint_class_default.set_editor_property('ArrayIndex_Leaves', index)
")))

(fn SetEditorStaticMeshAttributes
  (params (LoadedAsset%20Staticmesh :type Object%20Reference))
  (bind _returnvalue (EditorScripting|Asset|GetPathNameforLoadedAsset LoadedAsset%20Staticmesh :!class /Script/EditorScriptingUtilities.EditorAssetLibrary))
  (Python|Execution|ExecutePythonCommand (Utilities|String|Append "import unreal
" "" "
sm = '" _returnvalue "'


smload = unreal.load_asset(sm,None,True)


nanite_setting = smload.get_editor_property(\"NaniteSettings\")
nanite_setting.fallback_relative_error = 0.75
nanite_setting.fallback_percent_triangles = 80.0
# this applies the nanite settings
mesh_editor_ss = unreal.get_editor_subsystem(unreal.StaticMeshEditorSubsystem)

mesh_editor_ss.set_nanite_settings(mesh, nanite_setting, True)
")))

(fn SetEditorStaticMeshonBlueprint
  (params (Tree%20Blueprint%20Class :type Object%20Class%20Reference) (LoadedAsset%20Staticmesh :type Object%20Reference))
  (bind _returnvalue (EditorScripting|Asset|GetPathNameforLoadedAsset LoadedAsset%20Staticmesh :!class /Script/EditorScriptingUtilities.EditorAssetLibrary))
  (Python|Execution|ExecutePythonCommand (Utilities|String|Append "import unreal

asset = '" (Utilities|String|Append (Utilities|String|Append (Utilities|String|GetSubstring (Utilities|String|GetSubstring _returnvalue 0 (Utilities|String|FindSubstring _returnvalue "/" false true)) 0 (+ (Utilities|String|FindSubstring (Utilities|String|GetSubstring _returnvalue 0 (Utilities|String|FindSubstring _returnvalue "/" false true)) "/" false true) 1)) "Blueprint/") (Utilities|String|Append (Utilities|String|Append (Utilities|String|GetSubstring (Utilities|GetClassDisplayName Tree%20Blueprint%20Class) 0 (Utilities|String|FindSubstring (Utilities|GetClassDisplayName Tree%20Blueprint%20Class) "_" false true)) ".") (Utilities|GetClassDisplayName Tree%20Blueprint%20Class))) "'

sm = '" _returnvalue "'

bp = unreal.load_class(None, asset)

blueprint_class_default = unreal.get_default_object(bp)

smload = unreal.load_asset(sm,None,True)

blueprint_class_default.set_editor_property('StaticMesh', smload)
")))

(fn SetTreeDataFromDataAsset
  (bind _treedataasset (Variables|Default|GetTreeDataAsset)
    (:then
      (Variables|DontTouchOnlyRead|SetTreeData (Variables|/Game/NeverCook/ArtTools/TreeTools/TreeBakingData.TreeBakingData_C|GetTreeData _treedataasset)))
    (:else)))

(fn SpawnTreeAssets
  (CallFunction|SetTreeDatafromDataAsset)
  (CallFunction|DeleteallSpawnedActors)
  (Utilities|FlowControl|Sequence
    (:then_0
      (for (_array_element _array_index) (Variables|DontTouchOnlyRead|GetTreeData)
        (CallFunction|SpawnClassEvent (.Blueprint_Tree_15_72B637784D14B76A881869AEE9983FA7 (Utilities|Struct|BreakTreetoConvertData _array_element)) (Math|Transform|MakeTransform (+ (Math|Vector|MakeVector (* (Variables|Default|GetAssetSpacing) _array_index)) "0,-3000.000000,0")))))))

(fn SpawnClass
  (params (Class :type Actor%20Class%20Reference) (SpawnTransform :type Transform))
  (bind _returnvalue (Game|SpawnActorfromClass self Class SpawnTransform "Undefined" "SelectDefaultAtRuntime" 0 0))
  (Variables|/Script/Engine.Actor|SetTags (Utilities|Array|MakeArray "AutoSpawned") _returnvalue)
  (return _returnvalue))

(fn ConstructionScript
  (CallFunction|ExecuteContructionScript))

(fn WriteScalarValueToRendertarget
  (params (Index :type Float%20%28single-precision%29) (Value :type Float%20%28single-precision%29))
  (bind _returnvalue (Rendering|Material|CreateDynamicMaterialInstance 0 (Variables|Default|GetAttributeBakeMaterial) "None" "None" :!class /Script/Engine.KismetMaterialLibrary))
  (Rendering|Material|SetScalarParameterValue _returnvalue "Index" Index :!class /Script/Engine.MaterialInstanceDynamic)
  (Rendering|Material|SetScalarParameterValue _returnvalue "Value" Value :!class /Script/Engine.MaterialInstanceDynamic)
  (Rendering|Material|SetVectorParameterValue _returnvalue "Vector3" "(R=1.000000,G=1.000000,B=1.000000,A=1.000000)" :!class /Script/Engine.MaterialInstanceDynamic)
  (Rendering|DrawMaterialtoRenderTarget 0 (Variables|Default|GetRenderTarget_2D) _returnvalue)
  (Rendering|DrawMaterialtoRenderTarget 0 (Variables|Default|GetRenderTarget_2D_TempCopy) "/Game/Environments/Asteria/Foliage/Trees/MasterMaterials/BakeArrayTexture/M_CopyRT.M_CopyRT")
  (if (Variables|Default|GetDebug)
    (Development|PrintString 0 (Utilities|String|Append (Utilities|String|Append (Utilities|String|Append "Writing Scalar Index = " (Utilities|String|ToString(Float) (Variables|GetIndex))) "     Value = ") (Utilities|String|ToString(Float) (Variables|GetValue))))))

(fn WriteVectorValueToRendertarget
  (params (Index :type Float%20%28single-precision%29) (Value :type Linear%20Color%20Structure))
  (bind _returnvalue (Rendering|Material|CreateDynamicMaterialInstance 0 (Variables|Default|GetAttributeBakeMaterial) "None" "None" :!class /Script/Engine.KismetMaterialLibrary))
  (Rendering|Material|SetScalarParameterValue _returnvalue "Index" Index :!class /Script/Engine.MaterialInstanceDynamic)
  (Rendering|Material|SetScalarParameterValue _returnvalue "Value" 1.0 :!class /Script/Engine.MaterialInstanceDynamic)
  (Rendering|Material|SetVectorParameterValue _returnvalue "Vector3" Value :!class /Script/Engine.MaterialInstanceDynamic)
  (Rendering|DrawMaterialtoRenderTarget 0 (Variables|Default|GetRenderTarget_2D) _returnvalue)
  (Rendering|DrawMaterialtoRenderTarget 0 (Variables|Default|GetRenderTarget_2D_TempCopy) "/Game/Environments/Asteria/Foliage/Trees/MasterMaterials/BakeArrayTexture/M_CopyRT.M_CopyRT")
  (if (Variables|Default|GetDebug)
    (Development|PrintString 0 (Utilities|String|Append (Utilities|String|Append (Utilities|String|Append "Writing Vector Index = " (Utilities|String|ToString(Float) (Variables|GetIndex))) "     Value = ") (Utilities|String|ToString(LinearColor) (Variables|GetValue))))))
