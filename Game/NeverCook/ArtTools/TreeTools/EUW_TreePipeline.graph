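# Editor utility widget that bakes a tree impostor for a chosen static mesh.
# A details view exposes settings, and MakeNames derives folder and asset names from the mesh path.
# The Bake Impostor button spawns a temporary mesh actor and ImposterSprites baker, runs its steps, assigns the created material instance to the mesh, then destroys both after a delay.
# Collision-copy and bake-data loading chains are not wired to any event.
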EditorUtilityWidgetBlueprint /Game/NeverCook/ArtTools/TreeTools/EUW_TreePipeline
graphs [
 EventGraph {nodes [
   n1 = MakeTransform(Rotation="0, 0, 0")
   n2 = SpawnTransientActor!(
    execute
    WorldContextObject:Object Reference
    ActorClass:Actor Class Reference=/Game/NeverCook/ArtTools/Tech_Art/CopyCollision/BP_CopyCollision.BP_CopyCollision_C
    SpawnTransform:"Transform (by ref)"=@n1
    CollisionHandlingOverride:ESpawnActorCollisionHandlingMethod Enum=AlwaysSpawn
    OwnerActor:Actor Object Reference
    InstigatorPawn:Pawn Object Reference
   ) -> (then, SpawnedActor:BP Copy Collision Object Reference)
   > Set StaticMeshTarget!(
    execute
    StaticMeshTarget:Static Mesh Object Reference=Get Static Mesh() -> (Static Mesh:Static Mesh Object Reference)
    self:BP Copy Collision Object Reference=@n2
   ) -> (then, Output_Get:Static Mesh Object Reference)
   > Set CollisionMesh!(execute, CollisionMesh:Static Mesh Object Reference, self:BP Copy Collision Object Reference=@n2) -> (then, Output_Get:Static Mesh Object Reference)
   > CopyMeshCollision(self:BP Copy Collision Object Reference=@n2)
   n3 = Get BaseFolder() -> (BaseFolder:String)
   n4 = Get Treename() -> (Treename:String)
   n5 = LoadAsset(AssetPath=Append(A=Append(A=@n3, B=BakeData/), B=Append(A=@n4, B=_Bake)))
   > n6 = Cast To StaticMesh(Object=@n5)
   > n7 = Set StaticMeshWithImpostor(StaticMeshWithImpostor:Static Mesh Object Reference=@n6) -> (Output_Get:Static Mesh Object Reference)
   n8 = Get TargetMaterialID() -> (TargetMaterialID:int=0)
   SetMaterial!(
    execute=[@n12.then, @n7.then]
    self:Static Mesh Object Reference=Get Static Mesh()
    MaterialIndex:int=int + int(A=@n8, B=~)
    NewMaterial:Material Interface Object Reference=Get Created MIC!(self:BP Generate Imposter Sprites 2 FNAsteria Object Reference=Get CurrentImpostorBaker() -> (CurrentImpostorBaker:BP Generate Imposter Sprites 2 FNAsteria Object Reference)) -> (Created MIC:Material Instance Constant Object Reference)
   ) -> (then)
   > Delay!(execute, WorldContextObject:Object Reference, Duration=4) -> (then)
   > Destroy Actor(self=[Get CurrentImpostorBaker(), Get Temp Static Mesh Actor() -> (Temp Static Mesh Actor:Static Mesh Actor Object Reference)])
   "On Pressed (SetCollisionButton)"() {is=title K2Node_ComponentBoundEvent}
   > n9 = SpawnTransientActor(ActorClass=&StaticMeshActor, SpawnTransform=@n1, CollisionHandlingOverride=Undefined) -> (SpawnedActor:Static Mesh Actor Object Reference)
   > Set Temp Static Mesh Actor(Temp Static Mesh Actor:Static Mesh Actor Object Reference=@n9) -> (Output_Get:Static Mesh Actor Object Reference)
   > SetStaticMesh(self=Get StaticMeshComponent(self=@n9), NewMesh=Get Static Mesh())
   > n10 = SpawnTransientActor(
    ActorClass=/Game/ImpostorBaker/BP/BP_Generate_ImposterSprites_2_FNAsteria.BP_Generate_ImposterSprites_2_FNAsteria_C
    SpawnTransform=MakeTransform(Rotation="0, 0, 0")
    CollisionHandlingOverride=Undefined
   ) -> (SpawnedActor:BP Generate Imposter Sprites 2 FNAsteria Object Reference)
   > Set CurrentImpostorBaker(CurrentImpostorBaker:BP Generate Imposter Sprites 2 FNAsteria Object Reference=@n10) -> (Output_Get:BP Generate Imposter Sprites 2 FNAsteria Object Reference)
   n11 = Get Temp Static Mesh Actor()
   > Set Static Mesh Actor!(
    execute
    Static Mesh Actor:Static Mesh Actor Object Reference=@n11
    self:BP Generate Imposter Sprites 2 FNAsteria Object Reference=Get CurrentImpostorBaker()
   ) -> (then, Output_Get:Static Mesh Actor Object Reference)
   > Set TargetLODIndex!(
    execute
    TargetLODIndex:int=Get TargetLODIndex() -> (TargetLODIndex:int=0)
    self:BP Generate Imposter Sprites 2 FNAsteria Object Reference=Get CurrentImpostorBaker()
   ) -> (then, Output_Get:int=0)
   > Set TargetMaterialIDIndex!(execute, TargetMaterialIDIndex:int=@n8, self:BP Generate Imposter Sprites 2 FNAsteria Object Reference=Get CurrentImpostorBaker()) -> (then, Output_Get:int=0)
   > Set New Asset Path!(
    execute
    New Asset Path:String=Get ImpostorFolder() -> (ImpostorFolder:String)
    self:BP Generate Imposter Sprites 2 FNAsteria Object Reference=Get CurrentImpostorBaker()
   ) -> (then, Output_Get:String)
   > Set New Asset Texture Prefix!(
    execute
    New Asset Texture Prefix:String=Append(A=Append(A=T_, B=@n4), B=_Impostor)
    self:BP Generate Imposter Sprites 2 FNAsteria Object Reference=Get CurrentImpostorBaker()
   ) -> (then, Output_Get:String)
   > Set New Asset MIC Name!(
    execute
    New Asset MIC Name:String=Append(A=Append(A=MI_, B=@n4), B=_Impostor)
    self:BP Generate Imposter Sprites 2 FNAsteria Object Reference=Get CurrentImpostorBaker()
   ) -> (then, Output_Get:String)
   > Set New Asset ImpostorMeshName!(
    execute
    New Asset ImpostorMeshName:String=Append(A=@n4, B=_Impostor)
    self:BP Generate Imposter Sprites 2 FNAsteria Object Reference=Get CurrentImpostorBaker()
   ) -> (then, Output_Get:String)
   > Set ExistingMeshAssetName!(
    execute
    ExistingMeshAssetName:String=Get ExistingAssetName() -> (ExistingAssetName:String)
    self:BP Generate Imposter Sprites 2 FNAsteria Object Reference=Get CurrentImpostorBaker()
   ) -> (then, Output_Get:String)
   > SetupAll(self:BP Generate Imposter Sprites 2 FNAsteria Object Reference=Get CurrentImpostorBaker())
   > "1) Clear RTs"(self:BP Generate Imposter Sprites 2 FNAsteria Object Reference=Get CurrentImpostorBaker())
   > "2) RenderFrames"(self:BP Generate Imposter Sprites 2 FNAsteria Object Reference=Get CurrentImpostorBaker())
   > n12 = "3) Create Static Assets"(self:BP Generate Imposter Sprites 2 FNAsteria Object Reference=Get CurrentImpostorBaker())
   "On Property Changed (DetailsView_58)"!() -> (OutputDelegate, then, PropertyName:Name=None) {is=title K2Node_ComponentBoundEvent}
   > MakeNames()
   Event PreConstruct()
   > SetObject(self=Get DetailsView_58() -> (DetailsView_58:Details View Object Reference), NewObject=Self-Reference())
   > MakeNames()
   Event Construct()
   > MakeNames()
   Break Tree Converte Internal Data!(TreeConverte_InternalData:"Tree Converte Internal Data Structure (by ref)"="Get (a copy)"(
     Array:Array of Tree Converte Internal Data Structures=Get Tree_internalData() -> (Tree_internalData:Array of Tree Converte Internal Data Structures)
     Dimension 1=Get Index() -> (Index:int=0)
    ) -> (Output:Tree Converte Internal Data Structure)) -> (
    TrunkMaterial_2_#1:Material Instance Dynamic Object Reference
    LeavesMaterial_4_#2:Material Instance Dynamic Object Reference
    AssetBaseFolder_7_#3:String
    TreeName_9_#4:String
   ) {is=title K2Node_BreakStruct}
   n13 = Break Tree to Convert Data!(TreeToConvert_Data:"Tree To Convert Data Structure (by ref)"="Get (a copy)"(
     Array:Array of Tree To Convert Data Structures=Get TreeData() -> (TreeData:Array of Tree To Convert Data Structures)
     Dimension 1=Get Index() -> (Index=0)
    ) -> (Output:Tree To Convert Data Structure)) -> (
    OutputName_34_#5:String
    StaticMesh_Tree_Nanite_31_#6:Static Mesh Object Reference
    StaticMesh_Tree_FallBack_30_#7:Static Mesh Object Reference
    CustomLODs_38_#8:Array of Static Mesh Object References
    AddCustomLODMaterials_41_#9:Array of Material Interface Object References
    Blueprint_Tree_15_#10:Actor Class Reference
    Trunk_Material_Element_9_#11:int
    Leaves_Material_Element_11_#12:int
    Mesh_Material_IDforMask_23_#13:int
    CollisionMesh_27_#14:Static Mesh Object Reference
    UseCustomImpostor_43_#15:bool=~
    CustomImpostorMesh_46_#16:Static Mesh Object Reference
    FlipMaterialIDsForNaniteandFallbackMissmatch_48_#17:bool=~
    UseSplitTrunkCanopyImpostor_52_#18:bool=~
    StaticMesh_Canopy_55_#19:Static Mesh Object Reference
    StaticMesh_Trunk_56_#20:Static Mesh Object Reference
   ) {is=title K2Node_BreakStruct}
   Length(TargetArray:Array of Static Material Structures=Get StaticMaterials!(self:Static Mesh Object Reference=@n13.StaticMesh_Tree_FallBack_30_#7) -> (StaticMaterials:Array of Static Material Structures))
   n14 = Length(TargetArray:Array of Static Mesh Object References=@n13.CustomLODs_38_#8)
   SelectInt(A=int + int(A=@n14), B=-1, bPickA="integer > integer"(A=@n14))
   Append(A=Get BaseFolder(), B=Get Treename())
   Append(
    A=/Content
    B=GetSubstring!(SourceString:String=@n3, StartIndex:int=FindSubstring(SearchIn=@n3, Substring=/Game, bSearchFromEnd=true), Length:int=Len(S=@n3)) -> (ReturnValue:String)
   )
   Event Tick()
   Get EditorUtilitySubsystem()
   Get Static Mesh()
   Set Temp Static Mesh Actor(Temp Static Mesh Actor=@n11)
   Set ImpostorFolder(ImpostorFolder:String=Append(A=@n3, B=Impostor/)) -> (Output_Get:String)
  ]}
 MakeNames {nodes [
   MakeNames() {is=title K2Node_FunctionEntry}
   > n1 = Get!(execute:Exec) -> (then:Exec, else:Exec, Static Mesh:Static Mesh Object Reference)
   > n2 = GetPathNameForLoadedAsset(LoadedAsset=@n1)
   n3 = FindSubstring(SearchIn=@n2, Substring=/, bSearchFromEnd=true)
   n4 = GetSubstring(SourceString=@n2, StartIndex=0, Length=@n3)
   n5 = FindSubstring(SearchIn=@n4, Substring=/, bSearchFromEnd=true)
   > Set BaseFolder(BaseFolder:String=GetSubstring(SourceString=@n4, StartIndex=0, Length=int + int(A=@n5))) -> (Output_Get:String)
   n6 = GetSubstring(SourceString=@n2, StartIndex=int + int(A=@n3), Length=Len(S=@n2))
   n7 = GetSubstring(SourceString=@n6, StartIndex=0, Length=FindSubstring(SearchIn=@n6, Substring=., bSearchFromEnd=true))
   > Set Treename(Treename:String=@n7) -> (Output_Get:String)
   > Set ImpostorFolder(ImpostorFolder=Append(A=Get BaseFolder(), B=Impostor/))
   > Set ExistingAssetName(ExistingAssetName:String=@n2) -> (Output_Get:String)
   n8 = Get Treename()
   SelectString(A=@n7, B=@n8, bPickA="Equal Exactly (String)"(A=@n8))
  ]}
]
variables [
 BaseFolder {type=string, default=""}
 CurrentImpostorBaker {type=BP_Generate_ImposterSprites_2_FNAsteria_C, default=null}
 ExistingAssetName {type=string, default=""}
 ImpostorFolder {type=string, default=""}
 Static Mesh {type=StaticMesh, default=null}
 TargetLODIndex {type=integer, default=3}
 TargetMaterialID {type=integer, default=2}
 Temp Static Mesh Actor {type=StaticMeshActor, default=null}
 Treename {type=string, default=""}
]
widgets {
 info {rootWidgetClass=&GridPanel, widgetCount=6}
 widgets [
  GridPanel_0 {widgetClassPath=&GridPanel, widget=GridPanel_0}
  VerticalBox_115 {widgetClassPath=&VerticalBox, parent=GridPanel_0, widget=VerticalBox_115, slot {class=GridSlot}}
  DetailsView_58 {
   widgetClassPath=&DetailsView
   parent=VerticalBox_115
   widget=DetailsView_58
   properties {categoriesToShow [Meshes]}
   slot {class=VerticalBoxSlot}
  }
  TextBlock_51 {
   widgetClassPath=&TextBlock
   parent=VerticalBox_115
   widget=TextBlock_51
   properties {
    text=Make sure to Enable Impostor Baker Boolean in TreeMaterial
    font {
     bForceMonospaced=false
     bMaterialIsStencil=false
     fontMaterial=null
     fontObject=&Roboto
     letterSpacing=0
     monospacedWidth=1
     size=12
     skewAmount=0
     typefaceFontName=Bold
     outlineSettings {
      bApplyOutlineToDropShadows=false
      bMiteredCorners=false
      bSeparateFillAlpha=false
      outlineMaterial=null
      outlineSize=0
      outlineColor=<rgba 0 0 0 1>
     }
    }
   }
   slot {class=VerticalBoxSlot, properties {padding=<ltrb 10 10 10 10>}}
  }
  SetCollisionButton {
   widgetClassPath=&EditorUtilityButton
   parent=VerticalBox_115
   widget=SetCollisionButton
   properties {toolTipText=Set Collision}
   slot {class=VerticalBoxSlot, properties {horizontalAlignment=HAlign_Left}}
  }
  TextBlock_49 {
   widgetClassPath=&TextBlock
   parent=SetCollisionButton
   widget=TextBlock_49
   properties {text=Bake Impostor}
   slot {class=ButtonSlot, properties {horizontalAlignment=HAlign_Left}}
  }
 ]
}
